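# Sidebar text button widget, derived from FortTextButton, that changes its button style to match the current input method.
# Its borderless layout places the inherited input action icon beside uppercase, centered label text.
# On Construct it subscribes to CommonInputSubsystem input-method changes and immediately applies the current type, selecting separate mouse/keyboard, gamepad or touch styles.
# Destruct unbinds the listener.
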WidgetBlueprint /FortniteCoreUI/UI/Sidebar/Tags/SidebarCollectionInputTextButton
graphs [EventGraph {nodes [
   n1 = HandleInputMethodChanged!() -> (OutputDelegate, then, CurrentInputType:ECommonInputType Enum) {is=title K2Node_CustomEvent}
   > SetStyle!(
    execute
    self:Common Button Base Object Reference
    InStyle:Common Button Style Class Reference=Select!(
     MouseAndKeyboard:Common Button Style Class Reference=/FortUILibrary/Foundation/Buttons/ButtonStyles/TopBar/ButtonStyle-OutlineEmptyFill.ButtonStyle-OutlineEmptyFill_C
     Gamepad:Common Button Style Class Reference=/FortUILibrary/Foundation/Buttons/ButtonStyles/ButtonStyle-Clear.ButtonStyle-Clear_C
     Touch:Common Button Style Class Reference=/FortUILibrary/Foundation/Buttons/ButtonStyles/Athena/Athena_ButtonStyle_AngledBlueMenuButton.Athena_ButtonStyle_AngledBlueMenuButton_C
     Count:Common Button Style Class Reference=&ButtonStyle-Clear_C
     Index:ECommonInputType Enum=@n1.CurrentInputType
    ) -> (ReturnValue:Common Button Style Class Reference)
   ) -> (then)
   Event Construct()
   n2 = Create Event()
   n3 = Get CommonInputSubsystem!() -> (ReturnValue:Common Input Object Reference) {is=title K2Node_GetBlueprintContextEx}
   > Assign On Input Method Changed!(execute:Exec, self:Common Input Object Reference=@n3, Delegate:"Delegate (by ref)"=@n2) -> (then:Exec) {is=title K2Node_AssignDelegate}
   > HandleInputMethodChanged!(execute, self, CurrentInputType:ECommonInputType Enum=GetCurrentInputType(self=@n3)) -> (then) {is=title K2Node_CallFunction}
   Event Destruct()
   > Unbind Event from On Input Method Changed(self=@n3, Delegate=@n2)
  ]}]
variables []
widgets {
 info {inheritedWidgetCount=2, parentClass=/Script/FortniteUI.FortTextButton, rootWidgetClass=&Border, widgetCount=4}
 widgets [
  Border_0 {
   widgetClassPath=&Border
   widget=Border_0
   properties {
    background {
     bIsDynamicallyLoaded=false
     drawAs=NoDrawType
     imageType=NoImage
     mirroring=NoMirror
     resourceName=None
     resourceObject=null
     tiling=NoTile
     imageSize=<32 32>
     margin=<ltrb 0 0 0 0>
     outlineSettings {
      bUseBrushTransparency=false
      roundingType=HalfHeightRadius
      width=0
      color {colorUseRule=UseColor_Specified, specifiedColor=<rgba 0 0 0 0>}
      cornerRadii=<0 0 0 0>
     }
     tintColor {colorUseRule=UseColor_Specified, specifiedColor=<rgba 1 1 1 1>}
     uVRegion {bIsValid=false, max=<0 0>, min=<0 0>}
    }
    padding=<ltrb 8 4 8 4>
   }
  }
  HorizontalBox_0 {
   widgetClassPath=&HorizontalBox
   parent=Border_0
   widget=HorizontalBox_0
   slot {class=BorderSlot, properties {padding=<ltrb 8 4 8 4>}}
  }
  InputActionWidget {
   widgetClassPath=&CommonActionWidget
   bInherited=true
   parent=HorizontalBox_0
   widget=InputActionWidget
   slot {class=HorizontalBoxSlot, properties {verticalAlignment=VAlign_Center}}
  }
  Text_Content {
   widgetClassPath=&CommonTextBlock
   bInherited=true
   parent=HorizontalBox_0
   widget=Text_Content
   properties {
    justification=Center
    mobileFontSizeMultiplier=1.2
    scrollStyle=/FortUILibrary/Styles/Typography/Foundation/TextScrollStyle-NoFade.TextScrollStyle-NoFade_C
    style=/FortUILibrary/Styles/Typography/Foundation/TextStyle-BurbankSmall-S-White.TextStyle-BurbankSmall-S-White_C
    text="{Text}"
    textTransformPolicy=ToUpper
    font {
     bForceMonospaced=false
     bMaterialIsStencil=false
     fontMaterial=null
     fontObject=&BurbankSmall
     letterSpacing=0
     monospacedWidth=1
     size=17
     skewAmount=0
     typefaceFontName=Black
     outlineSettings {
      bApplyOutlineToDropShadows=false
      bMiteredCorners=false
      bSeparateFillAlpha=false
      outlineMaterial=null
      outlineSize=0
      outlineColor=<rgba 0 0 0 1>
     }
    }
    shadowOffset=<0 0>
   }
   slot {class=HorizontalBoxSlot, properties {verticalAlignment=VAlign_Center, padding=<ltrb 10 0 0 0>, size {sizeRule=Fill, value=1}}}
  }
 ]
}
