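# Widget Blueprint tag list for the sidebar social tag editor, laying out three social tag widgets in a horizontal row.
# The three tag slots are inherited bindings from SidebarSocialTagListBase, with no Blueprint logic beyond empty event stubs.
# CommonHardwareVisibilityBorders between the tags query Input.Touch, adding extra spacer width on touch devices.
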WidgetBlueprint /FortniteCoreUI/UI/Sidebar/Tags/SidebarSocialTagEditorTagList
graphs [EventGraph {nodes [
   Event PreConstruct()
   Event Construct()
   Event Tick()
  ]}]
variables []
widgets {
 info {inheritedWidgetCount=3, parentClass=/Script/FortniteUI.SidebarSocialTagListBase, rootWidgetClass=&HorizontalBox, widgetCount=8}
 widgets [
  HorizontalBox_481 {widgetClassPath=&HorizontalBox, widget=HorizontalBox_481}
  SidebarSocialTagWidget_1 {
   widgetClassPath=/Game/UI/Sidebar/Tags/SidebarSocialTagWidget
   bInherited=true
   bIsVariable=true
   parent=HorizontalBox_481
   widget=SidebarSocialTagWidget_1
   properties {visibility=Visible}
   slot {class=HorizontalBoxSlot}
  }
  CommonHardwareVisibilityBorder_1 {
   widgetClassPath=&CommonHardwareVisibilityBorder
   parent=HorizontalBox_481
   widget=CommonHardwareVisibilityBorder_1
   properties {visibilityQuery {
     autoDescription=" ANY( Input.Touch )"
     tokenStreamVersion=0
     userDescription=Add space if touch input
     queryTokenStream [0,1,1,1,0]
     tagDictionary [{tagName=Input.Touch}]
    }}
   slot {class=HorizontalBoxSlot}
  }
  Spacer_Mobile {
   widgetClassPath=&Spacer
   parent=CommonHardwareVisibilityBorder_1
   widget=Spacer_Mobile
   properties {size=<20 0>}
   slot {class=BorderSlot, properties {padding=<ltrb 0 0 0 0>}}
  }
  SidebarSocialTagWidget_2 {
   widgetClassPath=&SidebarSocialTagWidget
   bInherited=true
   bIsVariable=true
   parent=HorizontalBox_481
   widget=SidebarSocialTagWidget_2
   properties {visibility=Visible}
   slot {class=HorizontalBoxSlot, properties {padding=<ltrb 10 0 0 0>}}
  }
  CommonHardwareVisibilityBorder_2 {
   widgetClassPath=&CommonHardwareVisibilityBorder
   parent=HorizontalBox_481
   widget=CommonHardwareVisibilityBorder_2
   properties {visibilityQuery {
     autoDescription=" ANY( Input.Touch )"
     tokenStreamVersion=0
     userDescription=Add space if touch input
     queryTokenStream [0,1,1,1,0]
     tagDictionary [{tagName=Input.Touch}]
    }}
   slot {class=HorizontalBoxSlot}
  }
  Spacer_Mobile_1 {
   widgetClassPath=&Spacer
   parent=CommonHardwareVisibilityBorder_2
   widget=Spacer_Mobile_1
   properties {size=<20 0>}
   slot {class=BorderSlot, properties {padding=<ltrb 0 0 0 0>}}
  }
  SidebarSocialTagWidget_3 {
   widgetClassPath=&SidebarSocialTagWidget
   bInherited=true
   bIsVariable=true
   parent=HorizontalBox_481
   widget=SidebarSocialTagWidget_3
   properties {visibility=Visible}
   slot {class=HorizontalBoxSlot, properties {padding=<ltrb 10 0 0 0>}}
  }
 ]
}
