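# Sidebar screen for editing the player's social tags, with category tabs, a content switcher, and add, replace and remove prompts.
# Tab category changes recolor the title and the top and bottom background dynamic materials.
# The "What Are Tags?" button, ToggleTagsInfoTooltip or the full-screen input blocker toggles a tooltip, whose activation events show or hide that blocker and disable or enable tab actions.
# Shoulder-button tab prompts are bound in PreConstruct.
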WidgetBlueprint /FortniteCoreUI/UI/Sidebar/Tags/BP_SidebarCollectionScreenContainerEditTags
graphs [EventGraph {nodes [
   n1 = "On Button Clicked (TagsInfoInputBlocker)"!() -> (OutputDelegate, then, Button:Common Button Legacy Object Reference) {is=title K2Node_ComponentBoundEvent}
   n2 = Get TagsInfoTooltip() -> (TagsInfoTooltip:Common Activatable Widget Object Reference)
   n3 = Branch(execute=[@n1.then, @n6.then, @n18.then], Condition=IsActivated!(self:Common Activatable Widget Object Reference=@n2) -> (ReturnValue:bool))
   > n4 = DeactivateWidget(self=@n2)
   UpdateInputActionVisibility(execute=[@n4.then, @n5.then], self:Fort Sidebar Collection Screen Container Edit Tags Object Reference)
   n5 = ActivateWidget(execute=@n3.else, self=@n2)
   n6 = "On Button Clicked (TagsInfoButton)"!() -> (OutputDelegate, then, Button:Common Button Legacy Object Reference) {is=title K2Node_ComponentBoundEvent}
   n7 = BP_OnWhatAreTagsDeactivated() {is=title K2Node_CustomEvent}
   n8 = Get TagsInfoInputBlocker() -> (TagsInfoInputBlocker:Simple Common Button Object Reference)
   > SetVisibility(self=@n8, InVisibility=Hidden)
   > EnableTabActions(self:Athena Collection Screen Container Object Reference)
   n9 = BP_OnWhatAreTagsActivated() {is=title K2Node_CustomEvent}
   > SetVisibility(self=@n8, InVisibility=Visible)
   > DisableTabActions(self:Athena Collection Screen Container Object Reference)
   Event PreConstruct()
   > SetInputAction!(
    execute
    self:Common Action Widget Object Reference=Get NextTabAction() -> (NextTabAction:Common Action Widget Object Reference)
    InputActionRow:Data Table Row Handle Structure=Get NextTabInputAction() -> (NextTabInputAction:Data Table Row Handle Structure)
   ) -> (then)
   > SetInputAction(
    self=Get PreviousTabAction() -> (PreviousTabAction:Common Action Widget Object Reference)
    InputActionRow=Get PreviousTabInputAction() -> (PreviousTabInputAction:Data Table Row Handle Structure)
   )
   n10 = Get TagsInfoTooltip()
   > Bind Event to On Widget Deactivated(self:Common Activatable Widget Object Reference=@n10, Delegate=@n7) {is=title K2Node_AddDelegate}
   > Bind Event to On Widget Activated(self:Common Activatable Widget Object Reference=@n10, Delegate=@n9) {is=title K2Node_AddDelegate}
   n11 = Event On Tab Title Change!() -> (OutputDelegate, then, TagCategory:"ESocialTagCategory Enum (by ref)"=GameModes) {is=title K2Node_Event}
   > n12 = Get Sidebar Social Tag Category Structure!(execute, TagCategory:String=Enum to String!(Enumerator:ESocialTagCategory Enum=@n11.TagCategory) -> (ReturnValue:String)) -> (then, SidebarSocialTagCategoryStructure:Sidebar Social Tag Category Structure Structure) {is=title K2Node_CallFunction}
   n13 = Break Sidebar Social Tag Category Structure!(SidebarSocialTagCategoryStructure:"Sidebar Social Tag Category Structure Structure (by ref)"=@n12) -> (
    TagSkewBoxInnerColor_48_#1:Linear Color Structure
    TagSkewBoxInnerColorHovered_42_#2:Linear Color Structure
    TagSkewBoxInnerColorSelected_45_#3:Linear Color Structure
    TagSkewBoxOuterColor_41_#4:Linear Color Structure
    TagSkewBoxOuterColorSelected_47_#5:Linear Color Structure
    TagCategoryBaseColor_19_#6:Slate Color Structure
   ) {is=title K2Node_BreakStruct}
   > SetColorAndOpacity(
    self=Get Text_TabTitle() -> (Text_TabTitle:Common Text Object Reference)
    InColorAndOpacity=@n13.TagCategoryBaseColor_19_#6
   )
   > n14 = GetDynamicMaterial(self=Get Image_Background_Top() -> (Image_Background_Top:Image Object Reference))
   n15 = Break Slate Color!(SlateColor:"Slate Color Structure (by ref)"=@n13.TagCategoryBaseColor_19_#6) -> (SpecifiedColor:Linear Color Structure, ColorUseRule:ESlateColorStylingMode Enum) {is=title K2Node_BreakStruct}
   > SetVectorParameterValue(self=@n14, ParameterName=Color2, Value=@n15)
   > n16 = GetDynamicMaterial(self=Get Image_Background_Bottom() -> (Image_Background_Bottom:Image Object Reference))
   > SetVectorParameterValue(self=@n16, ParameterName=Color2, Value=@n15)
   n17 = Event On Show Tabs Set!() -> (OutputDelegate, then, bShowTabs:bool) {is=title K2Node_Event}
   > SetVisibility(
    self=Get HBox_Tabs() -> (HBox_Tabs:Horizontal Box Object Reference)
    InVisibility=Select(Option 0=Collapsed, Option 1=SelfHitTestInvisible, Index=@n17.bShowTabs)
   )
   n18 = Event ToggleTagsInfoTooltip() {is=title K2Node_Event}
   Event Construct()
   > Branch(Condition=IsMobileGame())
   > SetPadding!(
    execute
    self:Border Slot Object Reference=SlotAsBorderSlot!(Widget:Widget Object Reference=Get NextTabAction()) -> (ReturnValue:Border Slot Object Reference)
    InPadding_Left:float=-16
    InPadding_Top:float=0
    InPadding_Right:float=0
    InPadding_Bottom:float=0
   ) -> (then)
  ]}]
variables []
widgets {
 info {
  inheritedWidgetCount=11
  parentClass=/Script/FortniteUI.FortSidebarCollectionScreenContainerEditTags
  rootWidgetClass=&Overlay
  widgetCount=58
 }
 widgets [
  Overlay_HitTestable {widgetClassPath=&Overlay, widget=Overlay_HitTestable, properties {visibility=Visible}}
  GP_Content {
   widgetClassPath=&GridPanel
   parent=Overlay_HitTestable
   widget=GP_Content
   properties {columnFill [1], rowFill [0,1]}
   slot {class=OverlaySlot, properties {horizontalAlignment=HAlign_Fill, verticalAlignment=VAlign_Fill}}
  }
  WBP_CaptureForPostBufferUpdate {
   widgetClassPath=&WBP_CaptureForPostBufferUpdate
   bIsVariable=true
   parent=GP_Content
   slot {class=GridSlot, properties {layer=1, rowSpan=2}}
  }
  Image_Background_Top {
   widgetClassPath=&Image
   bIsVariable=true
   parent=GP_Content
   widget=Image_Background_Top
   properties {brush {
     bIsDynamicallyLoaded=false
     drawAs=Image
     imageType=NoImage
     mirroring=NoMirror
     resourceName=None
     resourceObject=/FortniteCoreUI/UI/Sidebar/Tags/Materials/M_UI_Sidebar_Tags_EditTagsHeader
     tiling=NoTile
     imageSize=<63 62>
     margin=<ltrb 0 0 0 0>
     outlineSettings {
      bUseBrushTransparency=false
      roundingType=HalfHeightRadius
      width=0
      color {colorUseRule=UseColor_Specified, specifiedColor=<rgba 0 0 0 0>}
      cornerRadii=<0 0 0 0>
     }
     tintColor {colorUseRule=UseColor_Specified, specifiedColor=<rgba 1 1 1 1>}
     uVRegion {bIsValid=false, max=<0 0>, min=<0 0>}
    }}
   slot {class=GridSlot}
  }
  VBox_Top {widgetClassPath=&VerticalBox, parent=GP_Content, widget=VBox_Top, slot {class=GridSlot, properties {layer=10}}}
  SafeZone_Title {
   widgetClassPath=&SafeZone
   bIsVariable=true
   parent=VBox_Top
   widget=SafeZone_Title
   properties {padBottom=false}
   slot {class=VerticalBoxSlot, properties {padding=<ltrb 80 56 80 0>}}
  }
  VerticalBox_ScreenText {widgetClassPath=&VerticalBox, parent=SafeZone_Title, widget=VerticalBox_ScreenText, slot {class=SafeZoneSlot}}
  HorizontalBox_0 {
   widgetClassPath=&HorizontalBox
   parent=VerticalBox_ScreenText
   widget=HorizontalBox_0
   slot {class=VerticalBoxSlot, properties {verticalAlignment=VAlign_Top}}
  }
  VerticalBox_0 {
   widgetClassPath=&VerticalBox
   parent=HorizontalBox_0
   widget=VerticalBox_0
   slot {class=HorizontalBoxSlot, properties {size {sizeRule=Fill, value=1}}}
  }
  RichText_Header {
   widgetClassPath=&CommonRichTextBlock
   bIsVariable=true
   parent=VerticalBox_0
   widget=RichText_Header
   properties {
    bOverrideDefaultStyle=true
    text=YOUR SOCIAL TAGS
    defaultTextStyleOverride {
     overflowPolicy=Clip
     transformPolicy=None
     colorAndOpacity {colorUseRule=UseColor_Specified, specifiedColor=<rgba 1 1 1 1>}
     font {
      bForceMonospaced=false
      bMaterialIsStencil=false
      fontMaterial=null
      fontObject=&Burbank
      letterSpacing=0
      monospacedWidth=1
      size=42
      skewAmount=0.2
      typefaceFontName=BurbankRegular-Black
      outlineSettings {
       bApplyOutlineToDropShadows=false
       bMiteredCorners=false
       bSeparateFillAlpha=false
       outlineMaterial=null
       outlineSize=0
       outlineColor=<rgba 0 0 0 1>
      }
     }
     highlightColor {colorUseRule=UseColor_Specified, specifiedColor=<rgba 0 0 0 1>}
     highlightShape {
      bIsDynamicallyLoaded=false
      drawAs=Image
      imageType=NoImage
      mirroring=NoMirror
      resourceName=None
      resourceObject=null
      tiling=NoTile
      imageSize=<32 32>
      margin=<ltrb 0 0 0 0>
      outlineSettings {
       bUseBrushTransparency=false
       roundingType=HalfHeightRadius
       width=0
       color {colorUseRule=UseColor_Specified, specifiedColor=<rgba 0 0 0 0>}
       cornerRadii=<0 0 0 0>
      }
      tintColor {colorUseRule=UseColor_Specified, specifiedColor=<rgba 1 1 1 1>}
      uVRegion {bIsValid=false, max=<0 0>, min=<0 0>}
     }
     selectedBackgroundColor {colorUseRule=UseColor_Foreground, specifiedColor=<rgba 1 0 1 1>}
     shadowColorAndOpacity=<rgba 0 0 0 1>
     shadowOffset=<0 0>
     strikeBrush {
      bIsDynamicallyLoaded=false
      drawAs=Image
      imageType=NoImage
      mirroring=NoMirror
      resourceName=None
      resourceObject=null
      tiling=NoTile
      imageSize=<32 32>
      margin=<ltrb 0 0 0 0>
      outlineSettings {
       bUseBrushTransparency=false
       roundingType=HalfHeightRadius
       width=0
       color {colorUseRule=UseColor_Specified, specifiedColor=<rgba 0 0 0 0>}
       cornerRadii=<0 0 0 0>
      }
      tintColor {colorUseRule=UseColor_Specified, specifiedColor=<rgba 1 1 1 1>}
      uVRegion {bIsValid=false, max=<0 0>, min=<0 0>}
     }
     underlineBrush {
      bIsDynamicallyLoaded=false
      drawAs=Image
      imageType=NoImage
      mirroring=NoMirror
      resourceName=None
      resourceObject=null
      tiling=NoTile
      imageSize=<32 32>
      margin=<ltrb 0 0 0 0>
      outlineSettings {
       bUseBrushTransparency=false
       roundingType=HalfHeightRadius
       width=0
       color {colorUseRule=UseColor_Specified, specifiedColor=<rgba 0 0 0 0>}
       cornerRadii=<0 0 0 0>
      }
      tintColor {colorUseRule=UseColor_Specified, specifiedColor=<rgba 1 1 1 1>}
      uVRegion {bIsValid=false, max=<0 0>, min=<0 0>}
     }
    }
   }
   slot {class=VerticalBoxSlot}
  }
  RichText_description {
   widgetClassPath=&CommonRichTextBlock
   bIsVariable=true
   parent=VerticalBox_0
   widget=RichText_description
   properties {
    bOverrideDefaultStyle=true
    mobileTextBlockScale=1.2
    renderOpacity=0.5
    text=Pick up to 3 tags that represent how you like to play Fortnite
    defaultTextStyleOverride {
     overflowPolicy=Clip
     transformPolicy=None
     colorAndOpacity {colorUseRule=UseColor_Specified, specifiedColor=<rgba 0.0331 0.402 1 1>}
     font {
      bForceMonospaced=false
      bMaterialIsStencil=false
      fontMaterial=null
      fontObject=&BurbankSmall
      letterSpacing=0
      monospacedWidth=1
      size=21
      skewAmount=0
      typefaceFontName=Bold
      outlineSettings {
       bApplyOutlineToDropShadows=false
       bMiteredCorners=false
       bSeparateFillAlpha=false
       outlineMaterial=null
       outlineSize=0
       outlineColor=<rgba 0 0 0 1>
      }
     }
     highlightColor {colorUseRule=UseColor_Specified, specifiedColor=<rgba 0 0 0 1>}
     highlightShape {
      bIsDynamicallyLoaded=false
      drawAs=Image
      imageType=NoImage
      mirroring=NoMirror
      resourceName=None
      resourceObject=null
      tiling=NoTile
      imageSize=<32 32>
      margin=<ltrb 0 0 0 0>
      outlineSettings {
       bUseBrushTransparency=false
       roundingType=HalfHeightRadius
       width=0
       color {colorUseRule=UseColor_Specified, specifiedColor=<rgba 0 0 0 0>}
       cornerRadii=<0 0 0 0>
      }
      tintColor {colorUseRule=UseColor_Specified, specifiedColor=<rgba 1 1 1 1>}
      uVRegion {bIsValid=false, max=<0 0>, min=<0 0>}
     }
     selectedBackgroundColor {colorUseRule=UseColor_Foreground, specifiedColor=<rgba 1 0 1 1>}
     shadowColorAndOpacity=<rgba 0 0 0 1>
     shadowOffset=<0 0>
     strikeBrush {
      bIsDynamicallyLoaded=false
      drawAs=Image
      imageType=NoImage
      mirroring=NoMirror
      resourceName=None
      resourceObject=null
      tiling=NoTile
      imageSize=<32 32>
      margin=<ltrb 0 0 0 0>
      outlineSettings {
       bUseBrushTransparency=false
       roundingType=HalfHeightRadius
       width=0
       color {colorUseRule=UseColor_Specified, specifiedColor=<rgba 0 0 0 0>}
       cornerRadii=<0 0 0 0>
      }
      tintColor {colorUseRule=UseColor_Specified, specifiedColor=<rgba 1 1 1 1>}
      uVRegion {bIsValid=false, max=<0 0>, min=<0 0>}
     }
     underlineBrush {
      bIsDynamicallyLoaded=false
      drawAs=Image
      imageType=NoImage
      mirroring=NoMirror
      resourceName=None
      resourceObject=null
      tiling=NoTile
      imageSize=<32 32>
      margin=<ltrb 0 0 0 0>
      outlineSettings {
       bUseBrushTransparency=false
       roundingType=HalfHeightRadius
       width=0
       color {colorUseRule=UseColor_Specified, specifiedColor=<rgba 0 0 0 0>}
       cornerRadii=<0 0 0 0>
      }
      tintColor {colorUseRule=UseColor_Specified, specifiedColor=<rgba 1 1 1 1>}
      uVRegion {bIsValid=false, max=<0 0>, min=<0 0>}
     }
    }
   }
   slot {class=VerticalBoxSlot}
  }
  CommonHardwareVisibilityBorder_187 {
   widgetClassPath=&CommonHardwareVisibilityBorder
   parent=HorizontalBox_0
   widget=CommonHardwareVisibilityBorder_187
   properties {visibilityQuery {
     autoDescription=" ANY( Input.Touch )"
     tokenStreamVersion=0
     userDescription=""
     queryTokenStream [0,1,1,1,0]
     tagDictionary [{tagName=Input.Touch}]
    }}
   slot {class=HorizontalBoxSlot}
  }
  SafeZone_90 {
   widgetClassPath=&SafeZone
   parent=CommonHardwareVisibilityBorder_187
   widget=SafeZone_90
   properties {padBottom=false, padLeft=false, padTop=false}
   slot {class=BorderSlot, properties {padding=<ltrb 0 0 0 0>}}
  }
  HeaderWidget {
   widgetClassPath=&/SidebarTagsEditorTagsHeader
   bInherited=true
   bIsVariable=true
   parent=VerticalBox_ScreenText
   widget=HeaderWidget
   properties {navigation=&self:WidgetTree.HeaderWidget.WidgetNavigation_0}
   slot {class=VerticalBoxSlot, properties {horizontalAlignment=HAlign_Left, verticalAlignment=VAlign_Top, padding=<ltrb 0 40 0 0>}}
  }
  Text_TabTitle {
   widgetClassPath=&CommonTextBlock
   bInherited=true
   parent=VerticalBox_ScreenText
   widget=Text_TabTitle
   properties {
    bAutoCollapseWithEmptyText=true
    text=Game Mode
    textTransformPolicy=ToUpper
    font {
     bForceMonospaced=false
     bMaterialIsStencil=false
     fontMaterial=/FortniteCoreUI/UI/Sidebar/Tags/Materials/M_UI_Sidebar_Tags_EditTagsText
     fontObject=&Burbank
     letterSpacing=0
     monospacedWidth=1
     size=96
     skewAmount=0.2
     typefaceFontName=BurbankRegular-Black
     outlineSettings {
      bApplyOutlineToDropShadows=false
      bMiteredCorners=false
      bSeparateFillAlpha=false
      outlineMaterial=null
      outlineSize=0
      outlineColor=<rgba 0 0 0 1>
     }
    }
    shadowOffset=<0 0>
   }
   slot {class=VerticalBoxSlot, properties {horizontalAlignment=HAlign_Right, verticalAlignment=VAlign_Bottom}}
  }
  Overlay_Tabs {
   widgetClassPath=&Overlay
   parent=VBox_Top
   widget=Overlay_Tabs
   slot {class=VerticalBoxSlot, properties {verticalAlignment=VAlign_Bottom}}
  }
  Spacer_114 {
   widgetClassPath=&Spacer
   parent=Overlay_Tabs
   widget=Spacer_114
   properties {size=<1 48>}
   slot {class=OverlaySlot, properties {horizontalAlignment=HAlign_Fill, verticalAlignment=VAlign_Fill}}
  }
  Image_Background_Tabs {
   widgetClassPath=&Image
   bIsVariable=true
   parent=Overlay_Tabs
   widget=Image_Background_Tabs
   properties {brush {
     bIsDynamicallyLoaded=false
     drawAs=Image
     imageType=NoImage
     mirroring=NoMirror
     resourceName=None
     resourceObject=/FortniteCoreUI/UI/Sidebar/Tags/Materials/M_UI_Sidebar_Tags_EditTagsTabs
     tiling=NoTile
     imageSize=<32 32>
     margin=<ltrb 0 0 0 0>
     outlineSettings {
      bUseBrushTransparency=false
      roundingType=HalfHeightRadius
      width=0
      color {colorUseRule=UseColor_Specified, specifiedColor=<rgba 0 0 0 0>}
      cornerRadii=<0 0 0 0>
     }
     tintColor {colorUseRule=UseColor_Specified, specifiedColor=<rgba 1 1 1 1>}
     uVRegion {bIsValid=false, max=<0 0>, min=<0 0>}
    }}
   slot {class=OverlaySlot, properties {horizontalAlignment=HAlign_Fill, verticalAlignment=VAlign_Fill}}
  }
  SafeZone_Tabs {
   widgetClassPath=&SafeZone
   bIsVariable=true
   parent=Overlay_Tabs
   widget=SafeZone_Tabs
   properties {padBottom=false, padTop=false}
   slot {class=OverlaySlot, properties {horizontalAlignment=HAlign_Fill, verticalAlignment=VAlign_Fill, padding=<ltrb 80 0 0 0>}}
  }
  FortVisualAttachment_0 {
   widgetClassPath=&FortVisualAttachment
   parent=SafeZone_Tabs
   widget=FortVisualAttachment_0
   properties {contentAnchor=<0 0.5>}
   slot {class=SafeZoneSlot, properties {hAlign=HAlign_Left, vAlign=VAlign_Center}}
  }
  HBox_Tabs {widgetClassPath=&HorizontalBox, bIsVariable=true, parent=FortVisualAttachment_0, widget=HBox_Tabs, slot {class=SizeBoxSlot}}
  CommonVisibilityWidget_PReviousTabAction {
   widgetClassPath=&CommonVisibilityWidgetLegacy
   parent=HBox_Tabs
   widget=CommonVisibilityWidget_PReviousTabAction
   properties {bShowForMouseAndKeyboard=false, verticalAlignment=VAlign_Center, padding=<ltrb 0 0 24 0>}
   slot {class=HorizontalBoxSlot, properties {horizontalAlignment=HAlign_Left, verticalAlignment=VAlign_Center}}
  }
  PreviousTabAction {
   widgetClassPath=&CommonActionWidget
   bIsVariable=true
   parent=CommonVisibilityWidget_PReviousTabAction
   widget=PreviousTabAction
   properties {inputActions [{dataTable=&InputActionDataTable, rowName=GenericLeftShoulder}]}
   slot {class=BorderSlot, properties {verticalAlignment=VAlign_Center, padding=<ltrb 0 0 24 0>}}
  }
  Border_TabButtons {
   widgetClassPath=&Border
   parent=HBox_Tabs
   widget=Border_TabButtons
   properties {
    background {
     bIsDynamicallyLoaded=false
     drawAs=NoDrawType
     imageType=NoImage
     mirroring=NoMirror
     resourceName=None
     resourceObject=null
     tiling=NoTile
     imageSize=<32 32>
     margin=<ltrb 0.5 0 0.5 0>
     outlineSettings {
      bUseBrushTransparency=false
      roundingType=HalfHeightRadius
      width=0
      color {colorUseRule=UseColor_Specified, specifiedColor=<rgba 0 0 0 0>}
      cornerRadii=<0 0 0 0>
     }
     tintColor {colorUseRule=UseColor_Specified, specifiedColor=<rgba 0.0086 0.1529 0.7913 1>}
     uVRegion {bIsValid=false, max=<0 0>, min=<0 0>}
    }
    padding=<ltrb 4 0 4 0>
   }
   slot {class=HorizontalBoxSlot, properties {horizontalAlignment=HAlign_Left, verticalAlignment=VAlign_Center}}
  }
  EntryBox_TabButtons {
   widgetClassPath=&DynamicEntryBox
   bInherited=true
   bIsVariable=true
   parent=Border_TabButtons
   widget=EntryBox_TabButtons
   properties {
    entryWidgetClass=/Game/UI/Sidebar/Tags/BP_SidebarCollectionScreenContainerTabButtonEditTags.BP_SidebarCollectionScreenContainerTabButtonEditTags_C
    entrySpacing=<20 0>
   }
   slot {class=BorderSlot, properties {padding=<ltrb 4 0 4 0>}}
  }
  CommonVisibilityWidget_NextTabAction {
   widgetClassPath=&CommonVisibilityWidgetLegacy
   parent=HBox_Tabs
   widget=CommonVisibilityWidget_NextTabAction
   properties {bShowForMouseAndKeyboard=false, verticalAlignment=VAlign_Center, padding=<ltrb 24 0 0 0>}
   slot {class=HorizontalBoxSlot, properties {horizontalAlignment=HAlign_Left, verticalAlignment=VAlign_Center}}
  }
  NextTabAction {
   widgetClassPath=&CommonActionWidget
   bIsVariable=true
   parent=CommonVisibilityWidget_NextTabAction
   widget=NextTabAction
   properties {inputActions [{dataTable=&InputActionDataTable, rowName=GenericRightShoulder}]}
   slot {class=BorderSlot, properties {verticalAlignment=VAlign_Center, padding=<ltrb 24 0 0 0>}}
  }
  Image_Background_Bottom {
   widgetClassPath=&Image
   bIsVariable=true
   parent=GP_Content
   widget=Image_Background_Bottom
   properties {brush {
     bIsDynamicallyLoaded=false
     drawAs=Image
     imageType=NoImage
     mirroring=NoMirror
     resourceName=None
     resourceObject=/FortniteCoreUI/UI/Sidebar/Tags/Materials/M_UI_Sidebar_Tags_EditTagsMode
     tiling=NoTile
     imageSize=<32 32>
     margin=<ltrb 0 0 0 0>
     outlineSettings {
      bUseBrushTransparency=false
      roundingType=HalfHeightRadius
      width=0
      color {colorUseRule=UseColor_Specified, specifiedColor=<rgba 0 0 0 0>}
      cornerRadii=<0 0 0 0>
     }
     tintColor {colorUseRule=UseColor_Specified, specifiedColor=<rgba 1 1 1 1>}
     uVRegion {bIsValid=false, max=<0 0>, min=<0 0>}
    }}
   slot {class=GridSlot, properties {row=1}}
  }
  Overlay_Bottom {widgetClassPath=&Overlay, parent=GP_Content, widget=Overlay_Bottom, slot {class=GridSlot, properties {layer=10, row=1}}}
  VerticalBox_1 {
   widgetClassPath=&VerticalBox
   parent=Overlay_Bottom
   widget=VerticalBox_1
   properties {clipping=ClipToBounds}
   slot {class=OverlaySlot, properties {horizontalAlignment=HAlign_Fill, verticalAlignment=VAlign_Fill, padding=<ltrb 0 0 0 96>}}
  }
  VisibilitySwitcher_Content {
   widgetClassPath=&CommonVisibilitySwitcher
   bInherited=true
   parent=VerticalBox_1
   widget=VisibilitySwitcher_Content
   properties {activeWidgetIndex=-1, bActivateFirstSlotOnAdding=true, bAutoActivateSlot=false}
   slot {class=VerticalBoxSlot, properties {size {sizeRule=Fill, value=1}}}
  }
  CommonHardwareVisibilityBorder_0 {
   widgetClassPath=&CommonHardwareVisibilityBorder
   parent=VerticalBox_1
   widget=CommonHardwareVisibilityBorder_0
   properties {visibilityQuery {
     autoDescription=" ANY( Input.Touch )"
     tokenStreamVersion=0
     userDescription=""
     queryTokenStream [0,1,1,1,0]
     tagDictionary [{tagName=Input.Touch}]
    }}
   slot {class=VerticalBoxSlot, properties {verticalAlignment=VAlign_Bottom}}
  }
  Spacer_Mobile {
   widgetClassPath=&Spacer
   bIsVariable=true
   parent=CommonHardwareVisibilityBorder_0
   widget=Spacer_Mobile
   properties {visibility=Visible, size=<0 15>}
   slot {class=BorderSlot, properties {padding=<ltrb 0 0 0 0>}}
  }
  SafeZone_ButtonsList {
   widgetClassPath=&SafeZone
   bIsVariable=true
   parent=Overlay_Bottom
   widget=SafeZone_ButtonsList
   properties {padBottom=false, padRight=false, padTop=false}
   slot {class=OverlaySlot, properties {horizontalAlignment=HAlign_Fill, verticalAlignment=VAlign_Bottom, padding=<ltrb 80 0 80 32>}}
  }
  Overlay_BottomBar {widgetClassPath=&HorizontalBox, parent=SafeZone_ButtonsList, widget=Overlay_BottomBar, slot {class=SafeZoneSlot}}
  HBox_LeftSideActions {
   widgetClassPath=&HorizontalBox
   parent=Overlay_BottomBar
   widget=HBox_LeftSideActions
   slot {class=HorizontalBoxSlot, properties {horizontalAlignment=HAlign_Left, verticalAlignment=VAlign_Center}}
  }
  HBox_AddTag {
   widgetClassPath=&HorizontalBox
   bInherited=true
   parent=HBox_LeftSideActions
   widget=HBox_AddTag
   slot {class=HorizontalBoxSlot, properties {horizontalAlignment=HAlign_Left, verticalAlignment=VAlign_Center, padding=<ltrb 0 0 50 0>}}
  }
  Action_AddTag {
   widgetClassPath=&CommonActionWidget
   parent=HBox_AddTag
   widget=Action_AddTag
   properties {visibility=SelfHitTestInvisible, inputActions [{dataTable=&/InputActionDataTable_TagsEditor, rowName=Select}]}
   slot {class=HorizontalBoxSlot, properties {padding=<ltrb 0 0 10 0>}}
  }
  Visibility_AddTag {
   widgetClassPath=&CommonVisibilityWidgetLegacy
   parent=HBox_AddTag
   widget=Visibility_AddTag
   properties {bShowForTouch=false, padding=<ltrb 0 0 0 0>}
   slot {class=HorizontalBoxSlot, properties {horizontalAlignment=HAlign_Right, verticalAlignment=VAlign_Center, size {sizeRule=Fill, value=1}}}
  }
  Text_AddTag {
   widgetClassPath=&CommonTextBlock
   parent=Visibility_AddTag
   widget=Text_AddTag
   properties {
    justification=Center
    scrollStyle=&TextScrollStyle-Base_C
    text=Add Tag
    textTransformPolicy=ToUpper
    font {
     bForceMonospaced=false
     bMaterialIsStencil=false
     fontMaterial=null
     fontObject=&BurbankSmall
     letterSpacing=0
     monospacedWidth=1
     size=18
     skewAmount=0
     typefaceFontName=Black
     outlineSettings {
      bApplyOutlineToDropShadows=false
      bMiteredCorners=false
      bSeparateFillAlpha=false
      outlineMaterial=null
      outlineSize=0
      outlineColor=<rgba 0 0 0 1>
     }
    }
    shadowOffset=<0 0>
   }
   slot {class=BorderSlot, properties {padding=<ltrb 0 0 0 0>}}
  }
  HBox_ReplaceTag {
   widgetClassPath=&HorizontalBox
   bInherited=true
   parent=HBox_LeftSideActions
   widget=HBox_ReplaceTag
   slot {class=HorizontalBoxSlot, properties {horizontalAlignment=HAlign_Left, verticalAlignment=VAlign_Center, padding=<ltrb 0 0 50 0>}}
  }
  Action_ReplaceTag {
   widgetClassPath=&CommonActionWidget
   parent=HBox_ReplaceTag
   widget=Action_ReplaceTag
   properties {visibility=SelfHitTestInvisible, inputActions [{dataTable=&InputActionDataTable_TagsEditor, rowName=Select}]}
   slot {class=HorizontalBoxSlot, properties {padding=<ltrb 0 0 10 0>}}
  }
  Visibility_ReplaceTag {
   widgetClassPath=&CommonVisibilityWidgetLegacy
   parent=HBox_ReplaceTag
   widget=Visibility_ReplaceTag
   properties {bShowForTouch=false, padding=<ltrb 0 0 0 0>}
   slot {class=HorizontalBoxSlot, properties {horizontalAlignment=HAlign_Right, verticalAlignment=VAlign_Center, size {sizeRule=Fill, value=1}}}
  }
  Text_ReplaceTag {
   widgetClassPath=&CommonTextBlock
   parent=Visibility_ReplaceTag
   widget=Text_ReplaceTag
   properties {
    justification=Center
    scrollStyle=&TextScrollStyle-Base_C
    text=Replace Tag
    textTransformPolicy=ToUpper
    font {
     bForceMonospaced=false
     bMaterialIsStencil=false
     fontMaterial=null
     fontObject=&BurbankSmall
     letterSpacing=0
     monospacedWidth=1
     size=18
     skewAmount=0
     typefaceFontName=Black
     outlineSettings {
      bApplyOutlineToDropShadows=false
      bMiteredCorners=false
      bSeparateFillAlpha=false
      outlineMaterial=null
      outlineSize=0
      outlineColor=<rgba 0 0 0 1>
     }
    }
    shadowOffset=<0 0>
   }
   slot {class=BorderSlot, properties {padding=<ltrb 0 0 0 0>}}
  }
  HBox_RemoveTag {
   widgetClassPath=&HorizontalBox
   bInherited=true
   parent=HBox_LeftSideActions
   widget=HBox_RemoveTag
   slot {class=HorizontalBoxSlot, properties {horizontalAlignment=HAlign_Left, verticalAlignment=VAlign_Center, padding=<ltrb 0 0 50 0>}}
  }
  RemoveTagButton {
   widgetClassPath=&/SidebarCollectionInputTextButton
   bInherited=true
   bIsVariable=true
   parent=HBox_RemoveTag
   widget=RemoveTagButton
   properties {
    bUseAllCaps=true
    defaultText=Remove Tag
    onButtonBaseClicked=(&self:WidgetTree.RemoveTagButton.HandleOnButtonClicked)
    onButtonBaseDoubleClicked=(&self:WidgetTree.RemoveTagButton.HandleOnButtonDoubleClicked)
    onButtonBaseHovered=(&self:WidgetTree.RemoveTagButton.HandleOnButtonHovered)
    onButtonBaseUnhovered=(&self:WidgetTree.RemoveTagButton.HandleOnButtonUnhovered)
    onSelectedChangedBase=(&self:WidgetTree.RemoveTagButton.HandleOnSelectedChanged)
    triggeringInputAction {dataTable=&InputActionDataTable_TagsEditor, rowName=RemoveTag}
   }
   slot {class=HorizontalBoxSlot}
  }
  Overlay_0 {
   widgetClassPath=&Overlay
   parent=Overlay_BottomBar
   widget=Overlay_0
   slot {
    class=HorizontalBoxSlot
    properties {horizontalAlignment=HAlign_Right, verticalAlignment=VAlign_Center, padding=<ltrb 32 0 32 0>, size {sizeRule=Fill, value=1}}
   }
  }
  Spacer_317 {
   widgetClassPath=&Spacer
   parent=Overlay_0
   widget=Spacer_317
   properties {size=<512 1>}
   slot {class=OverlaySlot, properties {horizontalAlignment=HAlign_Fill, verticalAlignment=VAlign_Fill}}
  }
  FortVisualAttachment_1 {
   widgetClassPath=&FortVisualAttachment
   parent=Overlay_0
   widget=FortVisualAttachment_1
   properties {contentAnchor=<0.5 1>}
   slot {class=OverlaySlot, properties {horizontalAlignment=HAlign_Center}}
  }
  TagsInfoTooltip {
   widgetClassPath=&/SidebarWhatAreTagsEditTagsTooltip
   bInherited=true
   bIsVariable=true
   parent=FortVisualAttachment_1
   widget=TagsInfoTooltip
   properties {visibility=Hidden}
   slot {class=SizeBoxSlot, properties {padding=<ltrb 0 0 0 32>}}
  }
  TagsInfoButton {
   widgetClassPath=&SidebarCollectionInputTextButton
   bIsVariable=true
   parent=Overlay_0
   widget=TagsInfoButton
   properties {
    bUseAllCaps=true
    clickMethod=MouseDown
    defaultText=What Are Tags?
    onButtonBaseClicked=(&self:WidgetTree.TagsInfoButton.HandleOnButtonClicked)
    onButtonBaseDoubleClicked=(&self:WidgetTree.TagsInfoButton.HandleOnButtonDoubleClicked)
    onButtonBaseHovered=(&self:WidgetTree.TagsInfoButton.HandleOnButtonHovered)
    onButtonBaseUnhovered=(&self:WidgetTree.TagsInfoButton.HandleOnButtonUnhovered)
    onSelectedChangedBase=(&self:WidgetTree.TagsInfoButton.HandleOnSelectedChanged)
    style=/FortUILibrary/Foundation/Buttons/ButtonStyles/TopBar/ButtonStyle-OutlineEmptyFill.ButtonStyle-OutlineEmptyFill_C
    hoveredSlateSoundOverride {resourceObject=/Game/Sounds/Fort_UI_Sounds/FrontEnd/DefaultSounds/PressedButtons/MSS_UI_Button_Hover_Default}
    triggeringInputAction {dataTable=&InputActionDataTable, rowName=Info}
   }
   slot {class=OverlaySlot, properties {horizontalAlignment=HAlign_Center, verticalAlignment=VAlign_Center}}
  }
  HBox_BackAction {
   widgetClassPath=&HorizontalBox
   parent=Overlay_BottomBar
   widget=HBox_BackAction
   slot {class=HorizontalBoxSlot, properties {horizontalAlignment=HAlign_Right, verticalAlignment=VAlign_Center}}
  }
  HardwareVisibilityBorder_Back {
   widgetClassPath=&CommonHardwareVisibilityBorder
   parent=HBox_BackAction
   widget=HardwareVisibilityBorder_Back
   properties {visibilityQuery {
     autoDescription=" NONE( Input.Touch )"
     tokenStreamVersion=0
     userDescription=""
     queryTokenStream [0,1,3,1,0]
     tagDictionary [{tagName=Input.Touch}]
    }}
   slot {class=HorizontalBoxSlot, properties {padding=<ltrb -8 -6 -8 -6>}}
  }
  Button_Back {
   widgetClassPath=&SidebarCollectionInputTextButton
   bInherited=true
   bIsVariable=true
   parent=HardwareVisibilityBorder_Back
   widget=Button_Back
   properties {
    bUseAllCaps=true
    clickMethod=MouseDown
    defaultText=BACK
    onButtonBaseClicked=(&self:WidgetTree.Button_Back.HandleOnButtonClicked)
    onButtonBaseDoubleClicked=(&self:WidgetTree.Button_Back.HandleOnButtonDoubleClicked)
    onButtonBaseHovered=(&self:WidgetTree.Button_Back.HandleOnButtonHovered)
    onButtonBaseUnhovered=(&self:WidgetTree.Button_Back.HandleOnButtonUnhovered)
    onSelectedChangedBase=(&self:WidgetTree.Button_Back.HandleOnSelectedChanged)
    style=&ButtonStyle-OutlineEmptyFill_C
    hoveredSlateSoundOverride {resourceObject=&MSS_UI_Button_Hover_Default}
    triggeringInputAction {dataTable=&InputActionDataTable, rowName=Back}
   }
   slot {class=BorderSlot, properties {padding=<ltrb 0 0 0 0>}}
  }
  SafeZone_TouchBack {
   widgetClassPath=&SafeZone
   bIsVariable=true
   parent=Overlay_HitTestable
   widget=SafeZone_TouchBack
   properties {padBottom=false, padLeft=false}
   slot {class=OverlaySlot, properties {horizontalAlignment=HAlign_Right, padding=<ltrb 0 16 16 0>}}
  }
  HardwareVisibilityBorder_TouchBack {
   widgetClassPath=&CommonHardwareVisibilityBorder
   parent=SafeZone_TouchBack
   widget=HardwareVisibilityBorder_TouchBack
   properties {visibilityQuery {
     autoDescription=" ANY( Input.Touch )"
     tokenStreamVersion=0
     userDescription=""
     queryTokenStream [0,1,1,1,0]
     tagDictionary [{tagName=Input.Touch}]
    }}
   slot {class=SafeZoneSlot}
  }
  Button_TouchBack {
   widgetClassPath=/FortUILibrary/Foundation/Buttons/CloseButton
   bInherited=true
   bIsVariable=true
   parent=HardwareVisibilityBorder_TouchBack
   widget=Button_TouchBack
   properties {
    onButtonBaseClicked=(&self:WidgetTree.Button_TouchBack.HandleOnButtonClicked)
    onButtonBaseDoubleClicked=(&self:WidgetTree.Button_TouchBack.HandleOnButtonDoubleClicked)
    onButtonBaseHovered=(&self:WidgetTree.Button_TouchBack.HandleOnButtonHovered)
    onButtonBaseUnhovered=(&self:WidgetTree.Button_TouchBack.HandleOnButtonUnhovered)
    onSelectedChangedBase=(&self:WidgetTree.Button_TouchBack.HandleOnSelectedChanged)
   }
   slot {class=BorderSlot, properties {padding=<ltrb 0 0 0 0>}}
  }
  TagsInfoInputBlocker {
   widgetClassPath=/Game/Athena/UI/Matchmaking/SimpleCommonButton
   bIsVariable=true
   parent=Overlay_HitTestable
   widget=TagsInfoInputBlocker
   properties {
    bDisplayInputActionWhenNotInteractable=false
    bIsFocusable=true
    onButtonBaseClicked=(&self:WidgetTree.TagsInfoInputBlocker.HandleOnButtonClicked)
    onButtonBaseDoubleClicked=(&self:WidgetTree.TagsInfoInputBlocker.HandleOnButtonDoubleClicked)
    onButtonBaseHovered=(&self:WidgetTree.TagsInfoInputBlocker.HandleOnButtonHovered)
    onButtonBaseUnhovered=(&self:WidgetTree.TagsInfoInputBlocker.HandleOnButtonUnhovered)
    onSelectedChangedBase=(&self:WidgetTree.TagsInfoInputBlocker.HandleOnSelectedChanged)
    style=/FortUILibrary/Foundation/Buttons/ButtonStyles/ButtonStyle-Clear.ButtonStyle-Clear_C
    visibility=Hidden
    triggeringInputAction {dataTable=&InputActionDataTable, rowName=GenericForward}
   }
   slot {class=OverlaySlot, properties {horizontalAlignment=HAlign_Fill, verticalAlignment=VAlign_Fill}}
  }
 ]
}
