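# Edit-tags sidebar list entry widget that shows a social tag as a skewed, category-colored pill with a checkmark.
# SetTag applies border and content colors from category, hover and checkmark state.
# On hover changes, UpdateTagState stores current and target colors, and a looping timer interpolates them using the StateChange animation's alpha.
# Collection-entry logic (new/high-score badges, randomized icon effects) remains, though its overlay is collapsed by default.
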WidgetBlueprint /FortniteCoreUI/UI/Sidebar/Tags/BP_SidebarCollectionListEntryEditTags
animations [
 Hover {
  binding_count=3
  length_seconds=0.15
  object=&self:Hover
  sequence {bindings [
    Text_ItemName {
     id=#1
     possessed_class=CommonTextBlock
     tracks [{
       type=MovieSceneColorTrack
       display_name=Color and Opacity
       sections [{
         type=MovieSceneColorSection
         end_frame=3
         start_frame=0
         channels [
          Color.R {type=MovieSceneScriptingFloatChannel, const [1,2]}
          Color.G {type=MovieSceneScriptingFloatChannel, const [1,2]}
          Color.B {type=MovieSceneScriptingFloatChannel, keys [[0,1],[3,0]]}
          Color.A {type=MovieSceneScriptingFloatChannel}
         ]
        }]
      }]
    }
    Border_Number {
     id=#2
     possessed_class=Border
     tracks [{
       type=MovieSceneColorTrack
       display_name=Content Color and Opacity
       sections [{
         type=MovieSceneColorSection
         channels [
          Color.R {type=MovieSceneScriptingFloatChannel}
          Color.G {type=MovieSceneScriptingFloatChannel}
          Color.B {type=MovieSceneScriptingFloatChannel}
          Color.A {type=MovieSceneScriptingFloatChannel, keys [[0,0.5],[3,0.75]]}
         ]
        }]
      }]
    }
    Image_Icon {
     id=#3
     possessed_class=FortLazyImage
     tracks [{
       type=MovieSceneWidgetMaterialTrack
       sections [{type=MovieSceneParameterSection, channels [hover {type=MovieSceneScriptingFloatChannel, keys [[0,0],[3,1]]}]}]
      }]
    }
   ]}
 }
 Selected {
  binding_count=5
  length_seconds=0.15
  object=&self:Selected
  sequence {bindings [
    HighScore {
     id=#4
     possessed_class=CommonBorder
     tracks [{
       type=MovieSceneWidgetMaterialTrack
       sections [{type=MovieSceneParameterSection, channels [HoverControl {type=MovieSceneScriptingFloatChannel, keys [[0,0],[3,1]]}]}]
      }]
    }
    Image_Icon {
     id=#3
     possessed_class=FortLazyImage
     tracks [{
       type=MovieSceneWidgetMaterialTrack
       sections [{type=MovieSceneParameterSection, channels [Selected {type=MovieSceneScriptingFloatChannel, keys [[0,0],[3,1]]}]}]
      }]
    }
    SizeBox_0 {
     id=#5
     possessed_class=SizeBox
     tracks [{
       type=MovieScene2DTransformTrack
       sections [{
         type=MovieScene2DTransformSection
         end_frame=3
         start_frame=0
         channels [
          Translation.X {type=MovieSceneScriptingFloatChannel}
          Translation.Y {type=MovieSceneScriptingFloatChannel}
          Angle {type=MovieSceneScriptingFloatChannel}
          Scale.X {type=MovieSceneScriptingFloatChannel, keys [[0,1],[3,1.05]]}
          Scale.Y {type=MovieSceneScriptingFloatChannel, keys [[0,1],[3,1.05]]}
          Shear.X {type=MovieSceneScriptingFloatChannel}
          Shear.Y {type=MovieSceneScriptingFloatChannel}
         ]
        }]
      }]
    }
    Text_Number {
     id=#6
     possessed_class=CommonTextBlock
     tracks [{
       type=MovieScene2DTransformTrack
       sections [{
         type=MovieScene2DTransformSection
         channels [
          Translation.X {type=MovieSceneScriptingFloatChannel, keys [[0,0],[3,-4]]}
          Translation.Y {type=MovieSceneScriptingFloatChannel, keys [[0,0],[3,6]]}
          Angle {type=MovieSceneScriptingFloatChannel}
          Scale.X {type=MovieSceneScriptingFloatChannel}
          Scale.Y {type=MovieSceneScriptingFloatChannel}
          Shear.X {type=MovieSceneScriptingFloatChannel}
          Shear.Y {type=MovieSceneScriptingFloatChannel}
         ]
        }]
      }]
    }
    HighScoreText {
     id=#7
     possessed_class=CommonTextBlock
     tracks [{
       type=MovieSceneWidgetMaterialTrack
       display_name=Font.FontMaterial
       sections [{type=MovieSceneParameterSection, channels [HoverControl {type=MovieSceneScriptingFloatChannel, keys [[0,0],[3,1]]}]}]
      }]
    }
   ]}
 }
 StateChange {
  binding_count=1
  length_seconds=0.1
  object=&self:StateChange
  sequence {bindings [BP_SidebarCollectionListEntryEditTags_C_88 {
     id=#8
     possessed_class=BP_SidebarCollectionListEntryEditTags_C
     tracks [{
       type=MovieSceneFloatTrack
       display_name=State Change Lerp Alpha
       sections [{type=MovieSceneFloatSection, end_frame=2, start_frame=0, channels [None {type=MovieSceneScriptingFloatChannel, keys [[0,0],[2,1]]}]}]
      }]
    }]}
 }
]
graphs [
 EventGraph {nodes [
   "On Input Icon Updated (InputActionWidget)"() {is=title K2Node_ComponentBoundEvent}
   > n1 = GetIconDynamicMaterial!(execute, self:Common Action Widget Object Reference=Get InputActionWidget() -> (InputActionWidget:Common Action Widget Object Reference)) -> (then, ReturnValue:Material Instance Dynamic Object Reference)
   > SetScalarParameterValue(self=@n1, ParameterName=Use Dark, Value=1)
   > SetScalarParameterValue(self=@n1, ParameterName=Outline Opacity, Value=0)
   InterpolateTagColors() {is=title K2Node_CustomEvent}
   n2 = Get StateChangeLerpAlpha() -> (StateChangeLerpAlpha:double)
   > InterpolateTagContentColors!(
    execute
    TagName:Text Object Reference=Get Text_TagName() -> (Text_TagName:Common Text Object Reference)
    CheckMark:Image Object Reference=Get Image_CheckMark() -> (Image_CheckMark:Image Object Reference)
    CurrentTagContentColor:Linear Color Structure=Get CurrentTagContentColorReference() -> (CurrentTagContentColorReference:Linear Color Structure)
    TargetTagContentColor:Linear Color Structure=Get TargetTagContentColorReference() -> (TargetTagContentColorReference:Linear Color Structure)
    HoveredTagContentLerpAlpha:double=@n2
   ) -> (then)
   > n3 = Get!(execute:Exec) -> (then:Exec, else:Exec, TagBorderMaterialReference:Material Instance Dynamic Object Reference)
   > InterpolateTagBorderColors!(
    execute
    TagBorderMaterial:Material Instance Dynamic Object Reference=@n3
    NameInnerColor:Name=Get Name_InnerColor() -> (Name_InnerColor:Name=None)
    CurrentInnerColor:Linear Color Structure=Get CurrentInnerColorReference() -> (CurrentInnerColorReference:Linear Color Structure)
    TargetInnerColor:Linear Color Structure=Get TargetInnerColorReference() -> (TargetInnerColorReference:Linear Color Structure)
    NameOuterColor:Name=Get Name_OuterColor() -> (Name_OuterColor:Name=None)
    CurrentOuterColor:Linear Color Structure=Get CurrentOuterColorReference() -> (CurrentOuterColorReference:Linear Color Structure)
    TargetOuterColor:Linear Color Structure=Get TargetOuterColorReference() -> (TargetOuterColorReference:Linear Color Structure)
    HoveredBorderColorLerpAlpha:double=@n2
    ColorLerpTimeHandle:"Timer Handle Structure (by ref)"=Get StateChangeLerpTimerHandle() -> (StateChangeLerpTimerHandle:Timer Handle Structure)
    WorldContext:Object Reference=Self-Reference()
   ) -> (then)
   n4 = Event On Collection List Item Set!() -> (OutputDelegate, then, TagName:"Text (by ref)", TagCategory:"ESocialTagCategory Enum (by ref)"=GameModes) {is=title K2Node_Event}
   SetTag!(
    execute=[@n4.then, @n30.then]
    self
    InTagName:String="To String (Text)"(InText=@n4.TagName)
    InTagCategory:ESocialTagCategory Enum=@n4.TagCategory
   ) -> (then)
   Event On Entry Reset() {is=title K2Node_Event}
   > SetVisibility(self=Get NewBang() -> (NewBang:Common Border Object Reference), InVisibility=Collapsed)
   > SetVisibility(self=Get HighScore() -> (HighScore:Common Border Object Reference), InVisibility=Collapsed)
   Event On Hovered()
   n5 = GetSelected!(self:Common Button Base Object Reference) -> (ReturnValue:bool)
   n6 = IsUsingTouch(PlayerController=GetOwningPlayer())
   > Branch(Condition=AND Boolean(A=NOT Boolean(A=@n5), B=NOT Boolean(A=@n6)))
   n7 = Get Hover() -> (Hover:Widget Animation Object Reference)
   > PlayAnimationForward(InAnimation=@n7)
   > n8 = Set Hovered(Hovered:bool=true) -> (Output_Get:bool)
   UpdateTagState(execute=[@n10.then, @n8.then])
   Event On Unhovered()
   > Branch(Condition=NOT Boolean(A=@n6))
   > n9 = Branch(Condition=GetSelected())
   PlayAnimationReverse(execute=@n9.else, InAnimation=@n7)
   > n10 = Set Hovered(Hovered=false)
   Event On Deselected()
   > Branch(Condition=Get IsSelected() -> (IsSelected:bool))
   > n11 = Branch(Condition=Get Hovered() -> (Hovered:bool))
   > n12 = PlayAnimationReverse(InAnimation=Get Hover())
   n13 = Get Selected() -> (Selected:Widget Animation Object Reference)
   PlayAnimationReverse(execute=[@n12.then, @n11.else], InAnimation=@n13)
   n14 = IsUsingTouch(PlayerController=GetOwningPlayer())
   > n15 = Branch(Condition=@n14)
   n16 = Set IsSelected(execute=[@n17.then, @n15.then], IsSelected:bool=false) -> (Output_Get:bool)
   Branch(execute=[@n33.then, @n16.then], Condition=GetSelected())
   > SetVisibility(self=Get NewBang(), InVisibility=Collapsed)
   n17 = Set Hovered(execute=@n15.else, Hovered=false)
   Event On Entry Image Set() {is=title K2Node_Event}
   > UpdateHoverParameter()
   n18 = Event On Collection State Set!() -> (OutputDelegate, then, CollectedState:EFortCollectedState Enum=Unknown) {is=title K2Node_Event}
   > RandomizeEffect()
   > SetVisibility(
    self=Get NewBang()
    InVisibility=Select!(
     Unknown:ESlateVisibility Enum=Collapsed
     New:ESlateVisibility Enum=SelfHitTestInvisible
     Known:ESlateVisibility Enum=Collapsed
     NewlyCollected:ESlateVisibility Enum=SelfHitTestInvisible
     Collected:ESlateVisibility Enum=Collapsed
     NewBest:ESlateVisibility Enum=SelfHitTestInvisible
     NewRecord:ESlateVisibility Enum=SelfHitTestInvisible
     NewLocation:ESlateVisibility Enum=Visible
     NewlyCompleted:ESlateVisibility Enum=Visible
     Complete:ESlateVisibility Enum=Visible
     Index:EFortCollectedState Enum=@n18.CollectedState
    ) -> (ReturnValue)
   )
   n19 = Make Slate Color(SpecifiedColor=(R=0,G=0.3348,B=1,A=1))
   n20 = Make Slate Color(SpecifiedColor=(R=0,G=0.1559,B=1,A=1))
   > SetColorAndOpacity(
    self=Get Text_Number() -> (Text_Number:Common Text Object Reference)
    InColorAndOpacity=Select(
     Unknown:Slate Color Structure=Make Slate Color(SpecifiedColor=(R=0,G=0.0578,B=0.6038,A=1))
     New:Slate Color Structure=@n20
     Known:Slate Color Structure=@n20
     NewlyCollected:Slate Color Structure=@n19
     Collected:Slate Color Structure=@n19
     NewBest:Slate Color Structure=@n19
     NewRecord:Slate Color Structure=@n19
     NewLocation:Slate Color Structure
     NewlyCompleted:Slate Color Structure
     Complete:Slate Color Structure
     Index=@n18.CollectedState
    ) -> (ReturnValue:Slate Color Structure)
   )
   > SetVisibility(
    self=Get HighScore()
    InVisibility=Select(
     Unknown:ESlateVisibility Enum=Collapsed
     New:ESlateVisibility Enum=Collapsed
     Known:ESlateVisibility Enum=Collapsed
     NewlyCollected:ESlateVisibility Enum=Collapsed
     Collected:ESlateVisibility Enum=Collapsed
     NewBest:ESlateVisibility Enum=SelfHitTestInvisible
     NewRecord:ESlateVisibility Enum=SelfHitTestInvisible
     NewLocation:ESlateVisibility Enum=Visible
     NewlyCompleted:ESlateVisibility Enum=Visible
     Complete:ESlateVisibility Enum=Visible
     Index=@n18.CollectedState
    ) -> (ReturnValue:ESlateVisibility Enum)
   )
   > n21 = GetDynamicMaterial(self=Get Image_Background() -> (Image_Background:Image Object Reference))
   > SetScalarParameterValue(
    self=@n21
    ParameterName=IsKnown
    Value=Select!(
     Option 0:double=1
     Option 1:double=0
     Index:bool="Equal (Enum)"(A:EFortCollectedState Enum=@n18.CollectedState, B:EFortCollectedState Enum=Unknown)
    ) -> (ReturnValue:double)
   )
   > SetScalarParameterValue(
    self=@n21
    ParameterName=IsDiscovered
    Value=Select!(
     Unknown:double=0
     New:double=0
     Known:double=0
     NewlyCollected:double=1
     Collected:double=1
     NewBest:double=1
     NewRecord:double=1
     NewLocation:double=0
     NewlyCompleted:double=0
     Complete:double=0
     Index:EFortCollectedState Enum=@n18.CollectedState
    ) -> (ReturnValue)
   )
   n22 = "Equal (Enum)"(A=@n18.CollectedState, B=NewBest)
   > Branch(Condition=OR Boolean(A=@n22, B="Equal (Enum)"(A=@n18.CollectedState, B=NewRecord)))
   > n23 = GetDynamicMaterial(self:Border Object Reference=Get HighScore())
   > SetVectorParameterValue(
    self=@n23
    ParameterName=MainColor
    Value=Select!(
     Option 0:Linear Color Structure=Get NewRecordBorderColor() -> (NewRecordBorderColor:Linear Color Structure)
     Option 1:Linear Color Structure=Get NewPersonalBestBorderColor() -> (NewPersonalBestBorderColor:Linear Color Structure)
     Index:bool=@n22
    ) -> (ReturnValue:Linear Color Structure)
   )
   n24 = Get HighScoreText() -> (HighScoreText:Common Text Object Reference)
   > "SetText (Text)"(
    self=@n24
    InText=Select(Option 0:Text=Get NewRecordText() -> (NewRecordText:Text), Option 1:Text=Get NewPersonalBestText() -> (NewPersonalBestText:Text), Index=@n22) -> (ReturnValue:Text)
   )
   > n25 = GetDynamicFontMaterial!(execute, self:Text Object Reference=@n24) -> (then, ReturnValue:Material Instance Dynamic Object Reference)
   n26 = Get NewPersonalBestTextColor() -> (NewPersonalBestTextColor:Slate Color Structure)
   n27 = Get NewRecordTextColor() -> (NewRecordTextColor:Slate Color Structure)
   > SetVectorParameterValue(
    self=@n25
    ParameterName=MainColor
    Value=Select(
     Option 0:Linear Color Structure=Break Slate Color!(SlateColor:"Slate Color Structure (by ref)"=@n27) -> (SpecifiedColor:Linear Color Structure, ColorUseRule:ESlateColorStylingMode Enum) {is=title K2Node_BreakStruct}
     Option 1:Linear Color Structure=Break Slate Color(SlateColor=@n26)
     Index=@n22
    ) -> (ReturnValue:Linear Color Structure)
   )
   Event PreConstruct()
   > RandomizeEffect()
   > n28 = GetDynamicMaterial(self:Image Object Reference=Get Image_Icon() -> (Image_Icon:Image Object Reference))
   > SetScalarParameterValue(self=@n28, ParameterName=ItemSize, Value=RandomFloatInRange(Min=0, Max=1))
   > Branch(Condition=IsMobileGame())
   > SetWidthOverride(self=Get SizeBox_0() -> (SizeBox_0:Size Box Object Reference), InWidthOverride=260)
   > SetUserSpecifiedScale!(
    execute
    self:Scale Box Object Reference=Get ScaleBox_Mobile() -> (ScaleBox_Mobile:Scale Box Object Reference)
    InUserSpecifiedScale:float=1.2
   ) -> (then)
   Event Destruct()
   > Clear and Invalidate Timer by Handle(Handle=Get StateChangeLerpTimerHandle())
   n29 = Event On Collection List Item checkmark visibiliy did change!() -> (OutputDelegate, then, bIsVisible:bool) {is=title K2Node_Event}
   > n30 = Set CheckmarkIs Visible(CheckmarkIs Visible:bool=@n29.bIsVisible) -> (Output_Get:bool)
   Event On Selected()
   > n31 = Branch(Condition=Get Hovered())
   PlayAnimationForward(execute=[@n35.then, @n31.then], InAnimation=@n13)
   > n32 = Branch(Condition=@n14)
   n33 = Set IsSelected(execute=[@n34.then, @n32.then], IsSelected=true)
   n34 = Set Hovered(execute=@n32.else, Hovered=true)
   n35 = PlayAnimationForward(execute=@n31.else, InAnimation=Get Hover())
  ]}
 GetCurrentTagBorderColors {nodes [
   GetCurrentTagBorderColors() {is=title K2Node_FunctionEntry}
   > n1 = Get()
   > n2 = Get Vector Parameter Value!(execute, self:Material Instance Dynamic Object Reference=@n1, ParameterName:Name=Get Name_InnerColor() -> (Name_InnerColor=None)) -> (then, ReturnValue:Linear Color Structure) {is=title K2Node_CallFunction}
   > Set CurrentInnerColorReference(CurrentInnerColorReference:Linear Color Structure=@n2) -> (Output_Get:Linear Color Structure)
   > n3 = Get Vector Parameter Value(self=@n1, ParameterName=Get Name_OuterColor() -> (Name_OuterColor=None))
   > Set CurrentOuterColorReference(CurrentOuterColorReference:Linear Color Structure=@n3) -> (Output_Get:Linear Color Structure)
  ]}
 RandomizeEffect {nodes [
   RandomizeEffect() {is=title K2Node_FunctionEntry}
   > n1 = GetDynamicMaterial(self=Get Image_Icon())
   > SetScalarParameterValue(self=@n1, ParameterName=RandomizeTime, Value=RandomFloatInRange(Min=-1, Max=1))
   > SetScalarParameterValue(self=@n1, ParameterName=RandomizeSpeed, Value=RandomFloatInRange(Min=0, Max=1))
   > SetScalarParameterValue(self=@n1, ParameterName=TimeMult, Value=RandomFloatInRange(Min=0.5, Max=1.25))
  ]}
 SetTag {nodes [
   n1 = SetTag!() -> (then, InTagName:String, InTagCategory:ESocialTagCategory Enum=GameModes) {is=title K2Node_FunctionEntry}
   > n2 = Sequence!(execute) -> (then_0, then_1, then_2:Exec, then_3:Exec)
   n3 = Get Sidebar Social Tag Category Structure!(execute=@n2.then_1, TagCategory:String=Enum to String!(Enumerator:ESocialTagCategory Enum=@n1.InTagCategory) -> (ReturnValue:String)) -> (then, SidebarSocialTagCategoryStructure:Sidebar Social Tag Category Structure Structure) {is=title K2Node_CallFunction}
   > Set TagCategoryStructureReference(TagCategoryStructureReference:Sidebar Social Tag Category Structure Structure=@n3) -> (Output_Get:Sidebar Social Tag Category Structure Structure)
   > n4 = Get Sidebar Social Tag Category Structure(TagCategory=Get TagCategory_Disabled() -> (TagCategory_Disabled:String))
   > Set DisabledTagCategoryStructureReference(DisabledTagCategoryStructureReference:Sidebar Social Tag Category Structure Structure=@n4) -> (Output_Get:Sidebar Social Tag Category Structure Structure)
   n5 = Get Text_TagName()
   "SetText (Text)"(execute=@n2.then_0, self=@n5, InText="To Text (String)"(InString=@n1))
   > SetScrollingEnabled!(execute, self:Common Text Object Reference=@n5, bInIsScrollingEnabled:bool=Get Hovered()) -> (then)
   > n6 = GetDynamicMaterial(self:Border Object Reference=Get Border_TagBorder() -> (Border_TagBorder:Border Object Reference))
   > Is Valid(InputObject=@n6)
   > Set TagBorderMaterialReference(TagBorderMaterialReference:Material Instance Dynamic Object Reference=@n6) -> (Output_Get:Material Instance Dynamic Object Reference)
   n7 = Get CheckmarkIs Visible() -> (CheckmarkIs Visible:bool)
   n8 = Get Hovered()
   n9 = SelectColor(
    A=Get TagContentColorBlack() -> (TagContentColorBlack:Linear Color Structure)
    B=SelectColor(
     A=Get TagContentColorWhite_Opacity90() -> (TagContentColorWhite_Opacity90:Linear Color Structure)
     B=Get TagContentColorWhite_Opacity60() -> (TagContentColorWhite_Opacity60:Linear Color Structure)
     bPickA=@n7
    )
    bPickA=@n8
   )
   SetColorAndOpacity(execute=@n2.then_2, self=Get Text_TagName(), InColorAndOpacity=Make Slate Color(SpecifiedColor=@n9))
   > SetColorAndOpacity(self:Image Object Reference=Get Image_CheckMark(), InColorAndOpacity:Linear Color Structure=@n9)
   n10 = GetDynamicMaterial(execute=@n2.then_3, self=Get Border_TagBorder())
   > Is Valid(InputObject=@n10)
   n11 = Break Sidebar Social Tag Category Structure!(SidebarSocialTagCategoryStructure:"Sidebar Social Tag Category Structure Structure (by ref)"=Get DisabledTagCategoryStructureReference() -> (DisabledTagCategoryStructureReference:Sidebar Social Tag Category Structure Structure)) -> (
    TagSkewBoxInnerColor_48_#9:Linear Color Structure
    TagSkewBoxOuterColor_41_#10:Linear Color Structure
   ) {is=title K2Node_BreakStruct}
   n12 = Break Sidebar Social Tag Category Structure!(SidebarSocialTagCategoryStructure=Get TagCategoryStructureReference() -> (TagCategoryStructureReference:Sidebar Social Tag Category Structure Structure)) -> (
    TagSkewBoxInnerColor_48_#9
    TagSkewBoxInnerColorHovered_42_#11:Linear Color Structure
    TagSkewBoxOuterColor_41_#10
    TagSkewBoxOuterColorSelected_47_#12:Linear Color Structure
   )
   > SetVectorParameterValue(
    self=@n10
    ParameterName=Get Name_InnerColor() -> (Name_InnerColor=None)
    Value=SelectColor(A=@n12.TagSkewBoxInnerColorHovered_42_#11, B=SelectColor(A=@n12, B=@n11, bPickA=@n7), bPickA=@n8)
   )
   > SetVectorParameterValue(
    self=@n10
    ParameterName=Get Name_OuterColor() -> (Name_OuterColor=None)
    Value=SelectColor(
     A=@n12.TagSkewBoxOuterColorSelected_47_#12
     B=SelectColor(A=@n12.TagSkewBoxOuterColor_41_#10, B=@n11.TagSkewBoxOuterColor_41_#10, bPickA=@n7)
     bPickA=@n8
    )
   )
  ]}
 UpdateHoverParameter {nodes [
   UpdateHoverParameter() {is=title K2Node_FunctionEntry}
   > n1 = GetDynamicMaterial(self:Image Object Reference=Get Image_Icon())
   n2 = GetSelected()
   > SetScalarParameterValue(self=@n1, ParameterName=hover, Value=SelectFloat(A=1, B=0, bPickA=@n2))
   > SetRenderTranslation(
    self=Get Text_Number()
    Translation=Select(Option 0:Vector 2D Structure, Option 1:Vector 2D Structure=(X=-4,Y=6), Index=@n2) -> (ReturnValue:Vector 2D Structure)
   )
   > SetContentColorAndOpacity!(
    execute
    self:Border Object Reference=Get Border_Number() -> (Border_Number:Border Object Reference)
    InContentColorAndOpacity:Linear Color Structure=Select(Option 0:Linear Color Structure=(R=1,G=1,B=1,A=0.5), Option 1:Linear Color Structure=(R=1,G=1,B=1,A=0.75), Index=@n2) -> (ReturnValue:Linear Color Structure)
   ) -> (then)
   > SetRenderScale!(
    execute
    self:Widget Object Reference=Get SizeBox_0()
    Scale:Vector 2D Structure=Select(Option 0:Vector 2D Structure=(X=1,Y=1), Option 1:Vector 2D Structure=(X=1.06,Y=1.06), Index=@n2) -> (ReturnValue:Vector 2D Structure)
   ) -> (then)
   > n3 = GetDynamicMaterial(self:Border Object Reference=Get HighScore())
   n4 = Select(Option 0:double=0, Option 1:double=1, Index=@n2) -> (ReturnValue:double)
   > SetScalarParameterValue(self=@n3, ParameterName=HoverControl, Value=@n4)
   > n5 = GetDynamicFontMaterial(self=Get HighScoreText())
   > SetScalarParameterValue(self=@n5, ParameterName=HoverControl, Value=@n4)
   > Set Hovered(Hovered=@n2)
   > Set IsSelected(IsSelected=@n2)
  ]}
 UpdateTagState {nodes [
   UpdateTagState() {is=title K2Node_FunctionEntry}
   > n1 = Sequence()
   > StopAllAnimations(self:User Widget Object Reference)
   > PlayAnimation(InAnimation=Get StateChange() -> (StateChange:Widget Animation Object Reference))
   Clear and Invalidate Timer by Handle(execute=@n1.then_1, Handle=Get StateChangeLerpTimerHandle())
   > GetCurrentTagBorderColors() {is=title K2Node_CallFunction}
   > Set CurrentTagContentColorReference(CurrentTagContentColorReference:Linear Color Structure=Get ColorAndOpacity!(self:Image Object Reference=Get Image_CheckMark()) -> (ColorAndOpacity:Linear Color Structure)) -> (Output_Get:Linear Color Structure)
   n2 = Get TagContentColorWhite_Opacity60()
   n3 = Get CheckmarkIs Visible()
   n4 = Get TagContentColorWhite_Opacity90()
   n5 = Get Hovered()
   > Set TargetTagContentColorReference(TargetTagContentColorReference:Linear Color Structure=SelectColor(A=Get TagContentColorBlack(), B=SelectColor(A=@n4, B=@n2, bPickA=@n3), bPickA=@n5)) -> (Output_Get:Linear Color Structure)
   n6 = Set Timer by Event(execute=@n1.then_3, Delegate=Create Event(), Time=Get StateChangeLerpTimeStep() -> (StateChangeLerpTimeStep:double), bLooping=true)
   > Set StateChangeLerpTimerHandle(StateChangeLerpTimerHandle:Timer Handle Structure=@n6) -> (Output_Get:Timer Handle Structure)
   > SetScrollingEnabled(self=Get Text_TagName(), bInIsScrollingEnabled=Get Hovered())
   n7 = Break Sidebar Social Tag Category Structure!(SidebarSocialTagCategoryStructure=Get DisabledTagCategoryStructureReference()) -> (TagSkewBoxInnerColor_48_#9, TagSkewBoxOuterColor_41_#10)
   n8 = Break Sidebar Social Tag Category Structure!(SidebarSocialTagCategoryStructure=Get TagCategoryStructureReference()) -> (
    TagSkewBoxInnerColor_48_#9
    TagSkewBoxInnerColorHovered_42_#11:Linear Color Structure
    TagSkewBoxOuterColor_41_#10
    TagSkewBoxOuterColorSelected_47_#12:Linear Color Structure
   )
   Set TargetInnerColorReference(
    execute=@n1.then_2
    TargetInnerColorReference:Linear Color Structure=SelectColor(A=@n8.TagSkewBoxInnerColorHovered_42_#11, B=SelectColor(A=@n8, B=@n7, bPickA=@n3), bPickA=@n5)
   ) -> (Output_Get:Linear Color Structure)
   > Set TargetOuterColorReference(TargetOuterColorReference:Linear Color Structure=SelectColor(
     A=@n8.TagSkewBoxOuterColorSelected_47_#12
     B=SelectColor(A=@n8.TagSkewBoxOuterColor_41_#10, B=@n7.TagSkewBoxOuterColor_41_#10, bPickA=@n3)
     bPickA=@n5
    )) -> (Output_Get:Linear Color Structure)
  ]}
]
variables [
 CheckmarkIs Visible {type=boolean}
 CurrentInnerColorReference {type=LinearColor, default=<rgba 0 0 0 1>}
 CurrentOuterColorReference {type=LinearColor, default=<rgba 0 0 0 1>}
 CurrentTagContentColorReference {type=LinearColor, default=<rgba 0 0 0 1>}
 DisabledTagCategoryStructureReference {
  type=SidebarSocialTagCategoryStructure
  default {
   tagCategoryIconMaterial=null
   tagCategoryBaseColor {colorUseRule=UseColor_Specified, specifiedColor=<rgba 1 0 1 1>}
   tagSkewBoxInnerColor=<rgba 0 0 0 0>
   tagSkewBoxInnerColorHovered=<rgba 0 0 0 0>
   tagSkewBoxInnerColorSelected=<rgba 0 0 0 1>
   tagSkewBoxOuterColor=<rgba 0 0 0 0>
   tagSkewBoxOuterColorSelected=<rgba 0 0 0 0>
  }
 }
 DisabledTagOpacity {type=number, default=0.6}
 EnabledTagOpacity {type=number, default=1}
 Hovered {type=boolean}
 IsSelected {type=boolean}
 Name_InnerColor {type=string, default=InnerColor}
 Name_OuterColor {type=string, default=OuterColor}
 Name_TagOpacity {type=string, default=Opacity}
 NewPersonalBestBorderColor {type=LinearColor, default=<rgba 0.7913 0.008 0.1845 1>}
 NewPersonalBestText {type=string, default="New Best!"}
 NewPersonalBestTextColor {type=SlateColor, default {colorUseRule=UseColor_Specified, specifiedColor=<rgba 1 1 1 1>}}
 NewRecordBorderColor {type=LinearColor, default=<rgba 0.9301 1 0.0097 1>}
 NewRecordText {type=string, default="High Score!"}
 NewRecordTextColor {type=SlateColor, default {colorUseRule=UseColor_Specified, specifiedColor=<rgba 0 0 0 1>}}
 StateChangeLerpAlpha {type=number}
 StateChangeLerpTimerHandle {type=TimerHandle}
 StateChangeLerpTimeStep {type=number, default=0.05}
 TagBorderMaterialReference {type=MaterialInstanceDynamic}
 TagCategory_Disabled {type=string, default=Disabled}
 TagCategoryStructureReference {
  type=SidebarSocialTagCategoryStructure
  default {
   tagCategoryIconMaterial=null
   tagCategoryBaseColor {colorUseRule=UseColor_Specified, specifiedColor=<rgba 1 0 1 1>}
   tagSkewBoxInnerColor=<rgba 0 0 0 0>
   tagSkewBoxInnerColorHovered=<rgba 0 0 0 0>
   tagSkewBoxInnerColorSelected=<rgba 0 0 0 1>
   tagSkewBoxOuterColor=<rgba 0 0 0 0>
   tagSkewBoxOuterColorSelected=<rgba 0 0 0 0>
  }
 }
 TagContentColorBlack {type=LinearColor, default=<rgba 0 0 0 1>}
 TagContentColorWhite_Opacity60 {type=LinearColor, default=<rgba 1 1 1 0.6>}
 TagContentColorWhite_Opacity90 {type=LinearColor, default=<rgba 1 1 1 0.9>}
 TargetInnerColorReference {type=LinearColor, default=<rgba 0 0 0 1>}
 TargetOuterColorReference {type=LinearColor, default=<rgba 0 0 0 1>}
 TargetTagContentColorReference {type=LinearColor, default=<rgba 0 0 0 1>}
]
widgets {
 info {
  inheritedWidgetCount=6
  parentClass=/Script/FortniteUI.FortSidebarCollectionListEntryEditTags
  rootWidgetClass=&ScaleBox
  widgetCount=24
 }
 widgets [
  ScaleBox_Mobile {
   widgetClassPath=&ScaleBox
   bIsVariable=true
   widget=ScaleBox_Mobile
   properties {stretch=UserSpecified, renderTransformPivot=<0 0>}
  }
  SizeBox_0 {
   widgetClassPath=&SizeBox
   bIsVariable=true
   parent=ScaleBox_Mobile
   widget=SizeBox_0
   properties {bOverride_HeightOverride=true, bOverride_WidthOverride=true, heightOverride=40, widthOverride=240}
   slot {class=ScaleBoxSlot, properties {horizontalAlignment=HAlign_Left, verticalAlignment=VAlign_Top}}
  }
  Overlay_0 {widgetClassPath=&Overlay, parent=SizeBox_0, widget=Overlay_0, slot {class=SizeBoxSlot}}
  Overlay_1 {
   widgetClassPath=&Overlay
   parent=Overlay_0
   widget=Overlay_1
   properties {visibility=Collapsed}
   slot {class=OverlaySlot, properties {horizontalAlignment=HAlign_Fill, verticalAlignment=VAlign_Fill}}
  }
  Image_Background {
   widgetClassPath=&Image
   bIsVariable=true
   parent=Overlay_1
   widget=Image_Background
   properties {brush {
     bIsDynamicallyLoaded=false
     drawAs=Image
     imageType=NoImage
     mirroring=NoMirror
     resourceName=None
     resourceObject=/Game/Athena/Collections/Art/Fishing/M_Ui_FishingCollectionENtry_Background_Inst
     tiling=NoTile
     imageSize=<32 32>
     margin=<ltrb 0 0 0 0>
     outlineSettings {
      bUseBrushTransparency=false
      roundingType=HalfHeightRadius
      width=0
      color {colorUseRule=UseColor_Specified, specifiedColor=<rgba 0 0 0 0>}
      cornerRadii=<0 0 0 0>
     }
     tintColor {colorUseRule=UseColor_Specified, specifiedColor=<rgba 1 1 1 1>}
     uVRegion {bIsValid=false, max=<0 0>, min=<0 0>}
    }}
   slot {class=OverlaySlot, properties {horizontalAlignment=HAlign_Fill, verticalAlignment=VAlign_Fill}}
  }
  Text_ItemName {
   widgetClassPath=&CommonTextBlock
   bInherited=true
   parent=Overlay_1
   widget=Text_ItemName
   properties {
    autoWrapText=true
    clipping=ClipToBounds
    justification=Center
    lineHeightPercentage=0.9
    mobileFontSizeMultiplier=1.2
    scrollStyle=&TextScrollStyle-Base_C
    text=Extra Very Long Item Name
    textTransformPolicy=ToUpper
    font {
     bForceMonospaced=false
     bMaterialIsStencil=false
     fontMaterial=null
     fontObject=&BurbankSmall
     letterSpacing=0
     monospacedWidth=1
     size=12
     skewAmount=0
     typefaceFontName=Black
     outlineSettings {
      bApplyOutlineToDropShadows=false
      bMiteredCorners=false
      bSeparateFillAlpha=false
      outlineMaterial=null
      outlineSize=0
      outlineColor=<rgba 0 0 0 1>
     }
    }
    renderTransform {angle=0, scale=<1 1>, shear=<-14 0>, translation=<0 0>}
    shadowOffset=<0 0>
   }
   slot {class=OverlaySlot, properties {horizontalAlignment=HAlign_Fill, verticalAlignment=VAlign_Center}}
  }
  Border_Number {
   widgetClassPath=&Border
   bIsVariable=true
   parent=Overlay_1
   widget=Border_Number
   properties {
    visibility=Collapsed
    background {
     bIsDynamicallyLoaded=false
     drawAs=NoDrawType
     imageType=NoImage
     mirroring=NoMirror
     resourceName=None
     resourceObject=null
     tiling=NoTile
     imageSize=<32 32>
     margin=<ltrb 0 0 0 0>
     outlineSettings {
      bUseBrushTransparency=false
      roundingType=HalfHeightRadius
      width=0
      color {colorUseRule=UseColor_Specified, specifiedColor=<rgba 0 0 0 0>}
      cornerRadii=<0 0 0 0>
     }
     tintColor {colorUseRule=UseColor_Specified, specifiedColor=<rgba 1 1 1 1>}
     uVRegion {bIsValid=false, max=<0 0>, min=<0 0>}
    }
    contentColorAndOpacity=<rgba 1 1 1 0.5>
   }
   slot {class=OverlaySlot}
  }
  Text_Number {
   widgetClassPath=&CommonTextBlock
   bInherited=true
   parent=Border_Number
   widget=Text_Number
   properties {
    text=01
    colorAndOpacity {colorUseRule=UseColor_Specified, specifiedColor=<rgba 0 0.15 1 1>}
    font {
     bForceMonospaced=false
     bMaterialIsStencil=false
     fontMaterial=null
     fontObject=&Burbank
     letterSpacing=0
     monospacedWidth=1
     size=24
     skewAmount=0
     typefaceFontName=BurbankRegular-Black
     outlineSettings {
      bApplyOutlineToDropShadows=false
      bMiteredCorners=false
      bSeparateFillAlpha=false
      outlineMaterial=null
      outlineSize=0
      outlineColor=<rgba 0 0 0 1>
     }
    }
    renderTransform {angle=0, scale=<1 1>, shear=<-12 0>, translation=<0 0>}
    shadowOffset=<0 0>
   }
   slot {class=BorderSlot}
  }
  Image_Icon {
   widgetClassPath=&FortLazyImage
   bInherited=true
   bIsVariable=true
   parent=Overlay_1
   widget=Image_Icon
   properties {
    visibility=Visible
    brush {
     bIsDynamicallyLoaded=false
     drawAs=Image
     imageType=NoImage
     mirroring=NoMirror
     resourceName=None
     resourceObject=/Game/Athena/Collections/Art/Fishing/M_Ui_CollectionNewEntry_Inst
     tiling=NoTile
     imageSize=<32 32>
     margin=<ltrb 0 0 0 0>
     outlineSettings {
      bUseBrushTransparency=false
      roundingType=HalfHeightRadius
      width=0
      color {colorUseRule=UseColor_Specified, specifiedColor=<rgba 0 0 0 0>}
      cornerRadii=<0 0 0 0>
     }
     tintColor {colorUseRule=UseColor_Specified, specifiedColor=<rgba 1 1 1 1>}
     uVRegion {bIsValid=false, max=<0 0>, min=<0 0>}
    }
   }
   slot {class=OverlaySlot, properties {horizontalAlignment=HAlign_Fill, verticalAlignment=VAlign_Fill}}
  }
  VerticalBox_1 {widgetClassPath=&VerticalBox, parent=Overlay_1, widget=VerticalBox_1, slot {class=OverlaySlot}}
  Image_FriendIcon {
   widgetClassPath=&Image
   bIsVariable=true
   parent=VerticalBox_1
   widget=Image_FriendIcon
   properties {
    visibility=Collapsed
    brush {
     bIsDynamicallyLoaded=false
     drawAs=Image
     imageType=NoImage
     mirroring=NoMirror
     resourceName=None
     resourceObject=/Game/Athena/Collections/Art/T_Ui_Friend_16
     tiling=NoTile
     imageSize=<16 16>
     margin=<ltrb 0 0 0 0>
     outlineSettings {
      bUseBrushTransparency=false
      roundingType=HalfHeightRadius
      width=0
      color {colorUseRule=UseColor_Specified, specifiedColor=<rgba 0 0 0 0>}
      cornerRadii=<0 0 0 0>
     }
     tintColor {colorUseRule=UseColor_Specified, specifiedColor=<rgba 0 1 1 1>}
     uVRegion {bIsValid=false, max=<0 0>, min=<0 0>}
    }
   }
   slot {class=VerticalBoxSlot, properties {horizontalAlignment=HAlign_Right, verticalAlignment=VAlign_Top, padding=<ltrb 12 0 10 0>}}
  }
  HighScore {
   widgetClassPath=&CommonBorder
   bIsVariable=true
   parent=Overlay_1
   widget=HighScore
   properties {
    visibility=Collapsed
    background {
     bIsDynamicallyLoaded=false
     drawAs=Box
     imageType=NoImage
     mirroring=NoMirror
     resourceName=None
     resourceObject=/Game/Athena/UI/ReplayBrowser/M_Ui_Skewed_Vfor9slices_Hover
     tiling=NoTile
     imageSize=<32 32>
     margin=<ltrb 0 0 0 0>
     outlineSettings {
      bUseBrushTransparency=false
      roundingType=HalfHeightRadius
      width=0
      color {colorUseRule=UseColor_Specified, specifiedColor=<rgba 0 0 0 0>}
      cornerRadii=<0 0 0 0>
     }
     tintColor {colorUseRule=UseColor_Specified, specifiedColor=<rgba 1 1 1 1>}
     uVRegion {bIsValid=false, max=<0 0>, min=<0 0>}
    }
    padding=<ltrb 3 3 3 1>
   }
   slot {class=OverlaySlot}
  }
  HighScoreText {
   widgetClassPath=&CommonTextBlock
   bIsVariable=true
   parent=HighScore
   widget=HighScoreText
   properties {
    justification=Center
    mobileFontSizeMultiplier=1.2
    scrollStyle=/Game/Athena/Prototype/Blueprints/Galileo/UI/TextScrollStyle-Base-Ticker.TextScrollStyle-Base-Ticker_C
    text="High Score!"
    textTransformPolicy=ToUpper
    font {
     bForceMonospaced=false
     bMaterialIsStencil=false
     fontMaterial=/Game/Athena/Collections/Art/M_Ui_textColorHover
     fontObject=&BurbankSmall
     letterSpacing=0
     monospacedWidth=1
     size=12
     skewAmount=0
     typefaceFontName=Black
     outlineSettings {
      bApplyOutlineToDropShadows=false
      bMiteredCorners=false
      bSeparateFillAlpha=false
      outlineMaterial=null
      outlineSize=0
      outlineColor=<rgba 0 0 0 1>
     }
    }
    renderTransform {angle=0, scale=<1 1>, shear=<-12 -1.5>, translation=<0 0>}
    shadowOffset=<0 0>
   }
   slot {class=BorderSlot, properties {padding=<ltrb 3 3 3 1>}}
  }
  NewBang {
   widgetClassPath=&CommonBorder
   bIsVariable=true
   parent=Overlay_1
   widget=NewBang
   properties {
    horizontalAlignment=HAlign_Center
    background {
     bIsDynamicallyLoaded=false
     drawAs=Box
     imageType=NoImage
     mirroring=NoMirror
     resourceName=None
     resourceObject=/Game/Athena/UI/ReplayBrowser/M_Ui_Skewed_Vfor9slices_Bang
     tiling=NoTile
     imageSize=<32 32>
     margin=<ltrb 0.5 0 0.5 0>
     outlineSettings {
      bUseBrushTransparency=false
      roundingType=HalfHeightRadius
      width=0
      color {colorUseRule=UseColor_Specified, specifiedColor=<rgba 0 0 0 0>}
      cornerRadii=<0 0 0 0>
     }
     tintColor {colorUseRule=UseColor_Specified, specifiedColor=<rgba 0.9301 1 0.0097 1>}
     uVRegion {bIsValid=false, max=<0 0>, min=<0 0>}
    }
    padding=<ltrb 8 6 8 6>
   }
   slot {class=OverlaySlot}
  }
  CommonTextBlock_209 {
   widgetClassPath=&CommonTextBlock
   parent=NewBang
   widget=CommonTextBlock_209
   properties {
    lineHeightPercentage=0.7429
    text="!"
    colorAndOpacity {colorUseRule=UseColor_Specified, specifiedColor=<rgba 0 0 0 1>}
    font {
     bForceMonospaced=false
     bMaterialIsStencil=false
     fontMaterial=null
     fontObject=&Burbank
     letterSpacing=0
     monospacedWidth=1
     size=18
     skewAmount=0
     typefaceFontName=BurbankRegular-Black
     outlineSettings {
      bApplyOutlineToDropShadows=false
      bMiteredCorners=false
      bSeparateFillAlpha=false
      outlineMaterial=null
      outlineSize=0
      outlineColor=<rgba 0 0 0 1>
     }
    }
    margin=<ltrb 0 -6 0 0>
    renderTransform {angle=0, scale=<1 1>, shear=<-12 0>, translation=<0 0>}
    shadowOffset=<0 0>
   }
   slot {class=BorderSlot, properties {horizontalAlignment=HAlign_Center, padding=<ltrb 8 6 8 6>}}
  }
  GridPanel_TagWidget {
   widgetClassPath=&GridPanel
   bIsVariable=true
   parent=Overlay_0
   widget=GridPanel_TagWidget
   properties {visibility=Visible, columnFill [0,1]}
   slot {class=OverlaySlot, properties {horizontalAlignment=HAlign_Fill, verticalAlignment=VAlign_Fill}}
  }
  Overlay_3 {widgetClassPath=&Overlay, parent=GridPanel_TagWidget, widget=Overlay_3, slot {class=GridSlot, properties {column=1}}}
  Border_TagBorder {
   widgetClassPath=&Border
   bIsVariable=true
   parent=Overlay_3
   widget=Border_TagBorder
   properties {
    background {
     bIsDynamicallyLoaded=false
     drawAs=Image
     imageType=NoImage
     mirroring=NoMirror
     resourceName=None
     resourceObject=/Game/UI/Sidebar/Tags/Materials/BoxSkew_SocialTag
     tiling=NoTile
     imageSize=<32 32>
     margin=<ltrb 0 0 0 0>
     outlineSettings {
      bUseBrushTransparency=false
      roundingType=HalfHeightRadius
      width=0
      color {colorUseRule=UseColor_Specified, specifiedColor=<rgba 0 0 0 0>}
      cornerRadii=<0 0 0 0>
     }
     tintColor {colorUseRule=UseColor_Specified, specifiedColor=<rgba 1 1 1 1>}
     uVRegion {bIsValid=false, max=<0 0>, min=<0 0>}
    }
    padding=<ltrb 8 6 12 6>
   }
   slot {class=OverlaySlot, properties {horizontalAlignment=HAlign_Fill, verticalAlignment=VAlign_Center}}
  }
  HorizontalBox_55 {
   widgetClassPath=&HorizontalBox
   parent=Border_TagBorder
   widget=HorizontalBox_55
   slot {class=BorderSlot, properties {padding=<ltrb 8 6 12 6>}}
  }
  Text_TagName {
   widgetClassPath=&CommonTextBlock
   bIsVariable=true
   parent=HorizontalBox_55
   widget=Text_TagName
   properties {
    bIsScrollingEnabled=false
    clipping=ClipToBounds
    mobileFontSizeMultiplier=1.13
    scrollStyle=&TextScrollStyle-Base_C
    text=Good Vibes
    textOverflowPolicy=Ellipsis
    colorAndOpacity {colorUseRule=UseColor_Specified, specifiedColor=<rgba 1 1 1 0.9>}
    font {
     bForceMonospaced=false
     bMaterialIsStencil=false
     fontMaterial=null
     fontObject=&BurbankSmall
     letterSpacing=0
     monospacedWidth=1
     size=16
     skewAmount=0.15
     typefaceFontName=Black
     outlineSettings {
      bApplyOutlineToDropShadows=false
      bMiteredCorners=false
      bSeparateFillAlpha=false
      outlineMaterial=null
      outlineSize=0
      outlineColor=<rgba 0 0 0 1>
     }
    }
    shadowOffset=<0 0>
   }
   slot {class=HorizontalBoxSlot, properties {verticalAlignment=VAlign_Center, padding=<ltrb 10 0 10 0>, size {sizeRule=Fill, value=1}}}
  }
  Image_CheckMark {
   widgetClassPath=&Image
   bInherited=true
   bIsVariable=true
   parent=HorizontalBox_55
   widget=Image_CheckMark
   properties {
    visibility=Collapsed
    brush {
     bIsDynamicallyLoaded=false
     drawAs=Image
     imageType=NoImage
     mirroring=NoMirror
     resourceName=None
     resourceObject=/Game/UI/Foundation/Textures/Icons/Checks/T-Icon-Check-64
     tiling=NoTile
     imageSize=<24 24>
     margin=<ltrb 0 0 0 0>
     outlineSettings {
      bUseBrushTransparency=false
      roundingType=HalfHeightRadius
      width=0
      color {colorUseRule=UseColor_Specified, specifiedColor=<rgba 0 0 0 0>}
      cornerRadii=<0 0 0 0>
     }
     tintColor {colorUseRule=UseColor_Specified, specifiedColor=<rgba 1 1 1 1>}
     uVRegion {bIsValid=false, max=<0 0>, min=<0 0>}
    }
   }
   slot {class=HorizontalBoxSlot, properties {horizontalAlignment=HAlign_Right, verticalAlignment=VAlign_Center, padding=<ltrb 0 0 5 2>}}
  }
  SizeBox_134 {
   widgetClassPath=&SizeBox
   parent=HorizontalBox_55
   widget=SizeBox_134
   properties {bOverride_HeightOverride=true, bOverride_WidthOverride=true, heightOverride=26, widthOverride=26}
   slot {class=HorizontalBoxSlot, properties {horizontalAlignment=HAlign_Right, verticalAlignment=VAlign_Center}}
  }
  InputActionWidget {
   widgetClassPath=&CommonActionWidget
   bInherited=true
   parent=SizeBox_134
   widget=InputActionWidget
   properties {
    visibility=SelfHitTestInvisible
    inputActions [{dataTable=&InputActionDataTable, rowName=Select}]
    progressMaterialBrush {
     bIsDynamicallyLoaded=false
     drawAs=Image
     imageType=NoImage
     mirroring=NoMirror
     resourceName=None
     resourceObject=null
     tiling=NoTile
     imageSize=<16 16>
     margin=<ltrb 0 0 0 0>
     outlineSettings {
      bUseBrushTransparency=false
      roundingType=HalfHeightRadius
      width=0
      color {colorUseRule=UseColor_Specified, specifiedColor=<rgba 0 0 0 0>}
      cornerRadii=<0 0 0 0>
     }
     tintColor {colorUseRule=UseColor_Specified, specifiedColor=<rgba 1 1 1 1>}
     uVRegion {bIsValid=false, max=<0 0>, min=<0 0>}
    }
   }
   slot {class=SizeBoxSlot}
  }
  Image_SecondaryIcon {widgetClassPath=&Image, bInherited=true, widget=null}
 ]
}
