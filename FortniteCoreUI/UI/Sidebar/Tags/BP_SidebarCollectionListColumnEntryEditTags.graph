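# Widget Blueprint for a group entry in the sidebar's edit-tags collection list, derived from FortSidebarCollectionListGroupEntry.
# It only configures the inherited FortDynamicEntryBox to hold a limited number of BP_SidebarCollectionListEntryEditTags entries with fixed spacing and a capped element size.
# The event graph is empty and no variables are added, so all behaviour comes from the native parent class and the entry widget.
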WidgetBlueprint /FortniteCoreUI/UI/Sidebar/Tags/BP_SidebarCollectionListColumnEntryEditTags
graphs [EventGraph {links [], nodes []}]
variables []
widgets {
 info {
  inheritedWidgetCount=1
  parentClass=/Script/FortniteUI.FortSidebarCollectionListGroupEntry
  rootWidgetClass=/Script/FortniteUI.FortDynamicEntryBox
  widgetCount=1
 }
 widgets [EntryBox_ListEntries {
   widgetClassPath=&FortDynamicEntryBox
   bInherited=true
   widget=EntryBox_ListEntries
   properties {
    entryWidgetClass=&/BP_SidebarCollectionListEntryEditTags.BP_SidebarCollectionListEntryEditTags_C
    maxElementSize=302
    maximumEntries=6
    numDesignerPreviewEntries=5
    entrySpacing=<32 0>
   }
  }]
}
