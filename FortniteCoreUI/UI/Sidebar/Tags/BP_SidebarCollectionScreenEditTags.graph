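# Sidebar screen widget for editing tags in a collection, built on FortSidebarCollectionScreenEditTags.
# A scrollable, multi-column CollectionListView of EditTags entries shares a visibility switcher with an "Updating Collection" spinner, toggled by the loading started and finished events.
# Activation plays Intro, which fades in the list and slides the info overlay container from the right.
# That container, the button bar and the season/name header default to collapsed.
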WidgetBlueprint /FortniteCoreUI/UI/Sidebar/Tags/BP_SidebarCollectionScreenEditTags
animations [Intro {
  binding_count=2
  length_seconds=0.25
  object=&self:Intro
  sequence {bindings [
    CollectionListView {
     id=#1
     possessed_class=AthenaCollectionListView
     tracks [{
       type=MovieSceneFloatTrack
       display_name=Render Opacity
       sections [{type=MovieSceneFloatSection, end_frame=5, start_frame=0, channels [None {type=MovieSceneScriptingFloatChannel, keys [[0,0],[5.0003,1]]}]}]
      }]
    }
    Content_InfoOverlayContainer {
     id=#2
     possessed_class=CommonBorder
     tracks [{
       type=MovieScene2DTransformTrack
       sections [{
         type=MovieScene2DTransformSection
         channels [
          Translation.X {type=MovieSceneScriptingFloatChannel, keys [[0,700],[5.0003,0]]}
          Translation.Y {type=MovieSceneScriptingFloatChannel}
          Angle {type=MovieSceneScriptingFloatChannel}
          Scale.X {type=MovieSceneScriptingFloatChannel}
          Scale.Y {type=MovieSceneScriptingFloatChannel}
          Shear.X {type=MovieSceneScriptingFloatChannel}
          Shear.Y {type=MovieSceneScriptingFloatChannel}
         ]
        }]
      }]
    }
   ]}
 }]
graphs [
 EventGraph {nodes [
   Event On Finished Loading() {is=title K2Node_Event}
   > SetActiveWidget(
    self=Get VisibilitySwitcher_Loading() -> (VisibilitySwitcher_Loading:Common Visibility Switcher Object Reference)
    Widget=Get Overlay_ListView() -> (Overlay_ListView:Overlay Object Reference)
   )
   Event On Started Loading() {is=title K2Node_Event}
   > SetActiveWidget(self=Get VisibilitySwitcher_Loading(), Widget=Get HorizontalBox_Loading() -> (HorizontalBox_Loading:Horizontal Box Object Reference))
   Event On Activated()
   > PlayAnimationForward(InAnimation=Get Intro())
  ]}
 Sequencer Events {nodes [
   n1 = PlayerProgressContainer_Event_0!() -> (OutputDelegate, then, PlayerProgressContainer:BP Athena Collection Player Progress Track Object Reference) {is=title K2Node_CustomEvent}
   > PlayIntro(self:BP Athena Collection Player Progress Track Object Reference=@n1.PlayerProgressContainer)
  ]}
]
variables []
widgets {
 info {
  inheritedWidgetCount=7
  parentClass=/Script/FortniteUI.FortSidebarCollectionScreenEditTags
  rootWidgetClass=&SafeZone
  widgetCount=25
 }
 widgets [
  SafeZone_2 {widgetClassPath=&SafeZone, bIsVariable=true, widget=SafeZone_2, properties {padBottom=false, padTop=false}}
  Overlay_176 {
   widgetClassPath=&Overlay
   parent=SafeZone_2
   widget=Overlay_176
   slot {class=SafeZoneSlot, properties {padding=<ltrb 0 32 0 0>, safeAreaScale=<ltrb 0.5 1 0.5 1>}}
  }
  VerticalBox_4 {
   widgetClassPath=&VerticalBox
   parent=Overlay_176
   widget=VerticalBox_4
   slot {class=OverlaySlot, properties {horizontalAlignment=HAlign_Fill, verticalAlignment=VAlign_Fill, padding=<ltrb 96 0 40 0>}}
  }
  VisibilitySwitcher_Loading {
   widgetClassPath=&CommonVisibilitySwitcher
   bIsVariable=true
   parent=VerticalBox_4
   widget=VisibilitySwitcher_Loading
   slot {class=VerticalBoxSlot, properties {padding=<ltrb 0 20 0 0>, size {sizeRule=Fill, value=1}}}
  }
  Overlay_ListView {
   widgetClassPath=&Overlay
   bIsVariable=true
   parent=VisibilitySwitcher_Loading
   widget=Overlay_ListView
   slot {class=CommonVisibilitySwitcherSlot}
  }
  SizeBox_2 {
   widgetClassPath=&SizeBox
   parent=Overlay_ListView
   widget=SizeBox_2
   properties {heightOverride=730}
   slot {class=OverlaySlot, properties {horizontalAlignment=HAlign_Fill, verticalAlignment=VAlign_Fill}}
  }
  Overlay_0 {widgetClassPath=&Overlay, parent=SizeBox_2, widget=Overlay_0, slot {class=SizeBoxSlot}}
  ScrollBox_List {
   widgetClassPath=/Script/FortniteUI.AthenaScrollBox
   bInherited=true
   bIsVariable=true
   parent=Overlay_0
   widget=ScrollBox_List
   properties {
    alwaysShowScrollbar=true
    alwaysShowScrollbarTrack=true
    bAnimateWheelScrolling=false
    bEnableClickThrough=true
    clipping=Inherit
    scrollWhenFocusChanges=InstantScroll
    scrollbarPadding=<ltrb 16 0 0 0>
    scrollbarThickness=<8 8>
   }
   slot {class=OverlaySlot, properties {horizontalAlignment=HAlign_Fill, verticalAlignment=VAlign_Fill}}
  }
  Overlay_1 {
   widgetClassPath=&Overlay
   parent=ScrollBox_List
   widget=Overlay_1
   slot {class=ScrollBoxSlot, properties {verticalAlignment=VAlign_Top}}
  }
  CollectionListView {
   widgetClassPath=/Script/FortniteUI.FortSidebarCollectionListView
   bInherited=true
   bIsVariable=true
   parent=Overlay_1
   widget=CollectionListView
   properties {
    bOverrideNumColumnsMobile=true
    clipping=Inherit
    entrySpacing=32
    entryWidgetClass=&/BP_SidebarCollectionListColumnEntryEditTags.BP_SidebarCollectionListColumnEntryEditTags_C
    navigation=&self:WidgetTree.CollectionListView.WidgetNavigation_0
    numColumns=6
    numColumnsMobile=5
    numDesignerPreviewEntries=20
    numRows=6
    orientation=Orient_Vertical
    visibility=SelfHitTestInvisible
   }
   slot {class=OverlaySlot, properties {horizontalAlignment=HAlign_Fill, padding=<ltrb 6 6 6 0>}}
  }
  HorizontalBox_Loading {
   widgetClassPath=&HorizontalBox
   bIsVariable=true
   parent=VisibilitySwitcher_Loading
   widget=HorizontalBox_Loading
   slot {class=CommonVisibilitySwitcherSlot, properties {horizontalAlignment=HAlign_Center, padding=<ltrb 96 64 96 64>}}
  }
  SizeBox_5069 {
   widgetClassPath=&SizeBox
   parent=HorizontalBox_Loading
   widget=SizeBox_5069
   properties {bOverride_HeightOverride=true, bOverride_WidthOverride=true, heightOverride=96, widthOverride=96}
   slot {class=HorizontalBoxSlot}
  }
  ProgressSpinner {
   widgetClassPath=&Image
   bIsVariable=true
   parent=SizeBox_5069
   widget=ProgressSpinner
   properties {brush {
     bIsDynamicallyLoaded=false
     drawAs=Image
     imageType=NoImage
     mirroring=NoMirror
     resourceName=None
     resourceObject=/FortUILibrary/UIKit/Throbber/MI_UIKit_Monochrome_Throbber
     tiling=NoTile
     imageSize=<64 64>
     margin=<ltrb 0 0 0 0>
     outlineSettings {
      bUseBrushTransparency=false
      roundingType=HalfHeightRadius
      width=0
      color {colorUseRule=UseColor_Specified, specifiedColor=<rgba 0 0 0 0>}
      cornerRadii=<0 0 0 0>
     }
     tintColor {colorUseRule=UseColor_Specified, specifiedColor=<rgba 1 1 1 1>}
     uVRegion {bIsValid=false, max=<0 0>, min=<0 0>}
    }}
   slot {class=SizeBoxSlot, properties {horizontalAlignment=HAlign_Center, verticalAlignment=VAlign_Center}}
  }
  TextMessage {
   widgetClassPath=&CommonTextBlock
   bIsVariable=true
   parent=HorizontalBox_Loading
   widget=TextMessage
   properties {
    justification=Center
    text=Updating Collection
    colorAndOpacity {colorUseRule=UseColor_Specified, specifiedColor=<rgba 0.2159 0.6584 1 1>}
    font {
     bForceMonospaced=false
     bMaterialIsStencil=false
     fontMaterial=null
     fontObject=&BurbankSmall
     letterSpacing=0
     monospacedWidth=1
     size=24
     skewAmount=0
     typefaceFontName=Bold
     outlineSettings {
      bApplyOutlineToDropShadows=false
      bMiteredCorners=false
      bSeparateFillAlpha=false
      outlineMaterial=null
      outlineSize=0
      outlineColor=<rgba 0 0 0 1>
     }
    }
    renderTransform {angle=0, scale=<1 1>, shear=<-12 0>, translation=<0 0>}
    shadowOffset=<0 0>
   }
   slot {class=HorizontalBoxSlot, properties {horizontalAlignment=HAlign_Center, verticalAlignment=VAlign_Center, padding=<ltrb 10 0 0 0>}}
  }
  Content_InfoOverlayContainer {
   widgetClassPath=&CommonBorder
   bInherited=true
   parent=Overlay_176
   widget=Content_InfoOverlayContainer
   properties {visibility=Collapsed, padding=<ltrb 0 0 0 0>}
   slot {class=OverlaySlot, properties {horizontalAlignment=HAlign_Right, verticalAlignment=VAlign_Fill, padding=<ltrb 40 30 40 16>}}
  }
  SafeZone_0 {
   widgetClassPath=&SafeZone
   parent=Overlay_176
   widget=SafeZone_0
   properties {padLeft=false, padRight=false, padTop=false, visibility=Collapsed}
   slot {class=OverlaySlot, properties {verticalAlignment=VAlign_Bottom, padding=<ltrb 16 16 16 16>}}
  }
  HorizontalBox_4 {
   widgetClassPath=&HorizontalBox
   parent=SafeZone_0
   widget=HorizontalBox_4
   slot {class=SafeZoneSlot, properties {hAlign=HAlign_Right}}
  }
  Button_ToggleInfoOverlay {
   widgetClassPath=&IconTextButton
   bInherited=true
   bIsVariable=true
   parent=HorizontalBox_4
   widget=Button_ToggleInfoOverlay
   properties {
    buttonText=HIde details
    minHeight=45
    onButtonBaseClicked=(&self:WidgetTree.Button_ToggleInfoOverlay.HandleOnButtonClicked)
    onButtonBaseDoubleClicked=(&self:WidgetTree.Button_ToggleInfoOverlay.HandleOnButtonDoubleClicked)
    onButtonBaseHovered=(&self:WidgetTree.Button_ToggleInfoOverlay.HandleOnButtonHovered)
    onButtonBaseUnhovered=(&self:WidgetTree.Button_ToggleInfoOverlay.HandleOnButtonUnhovered)
    onSelectedChangedBase=(&self:WidgetTree.Button_ToggleInfoOverlay.HandleOnSelectedChanged)
    style=/FortUILibrary/Foundation/Buttons/ButtonStyles/Athena/Athena_ButtonStyle_AngledBlueMenuButton_Small1.Athena_ButtonStyle_AngledBlueMenuButton_Small1_C
    triggeringInputAction {dataTable=&InputActionDataTable, rowName=GenericForward}
   }
   slot {class=HorizontalBoxSlot}
  }
  Button_Back {
   widgetClassPath=&IconTextButton
   bInherited=true
   bIsVariable=true
   parent=HorizontalBox_4
   widget=Button_Back
   properties {
    buttonText=Back
    minHeight=45
    onButtonBaseClicked=(&self:WidgetTree.Button_Back.HandleOnButtonClicked)
    onButtonBaseDoubleClicked=(&self:WidgetTree.Button_Back.HandleOnButtonDoubleClicked)
    onButtonBaseHovered=(&self:WidgetTree.Button_Back.HandleOnButtonHovered)
    onButtonBaseUnhovered=(&self:WidgetTree.Button_Back.HandleOnButtonUnhovered)
    onSelectedChangedBase=(&self:WidgetTree.Button_Back.HandleOnSelectedChanged)
    style=/FortUILibrary/Foundation/Buttons/ButtonStyles/Athena/Athena_ButtonStyle_MinimalHUD.Athena_ButtonStyle_MinimalHUD_C
    visibility=Collapsed
    triggeringInputAction {dataTable=&InputActionDataTable, rowName=Back}
   }
   slot {class=HorizontalBoxSlot, properties {padding=<ltrb 16 0 0 0>}}
  }
  HorizontalBox_0 {
   widgetClassPath=&HorizontalBox
   parent=Overlay_176
   widget=HorizontalBox_0
   properties {visibility=Collapsed}
   slot {class=OverlaySlot, properties {padding=<ltrb 0 0 0 24>}}
  }
  VerticalBox_0 {
   widgetClassPath=&VerticalBox
   parent=HorizontalBox_0
   widget=VerticalBox_0
   slot {class=HorizontalBoxSlot, properties {padding=<ltrb 256 0 0 0>}}
  }
  Text_Season {
   widgetClassPath=&CommonTextBlock
   parent=VerticalBox_0
   widget=Text_Season
   properties {
    lineHeightPercentage=0.85
    text="Season ######"
    textTransformPolicy=ToUpper
    colorAndOpacity {colorUseRule=UseColor_Specified, specifiedColor=<rgba 0.2122 0.6584 1 1>}
    font {
     bForceMonospaced=false
     bMaterialIsStencil=false
     fontMaterial=null
     fontObject=&BurbankSmall
     letterSpacing=0
     monospacedWidth=1
     size=18
     skewAmount=0
     typefaceFontName=Black
     outlineSettings {
      bApplyOutlineToDropShadows=false
      bMiteredCorners=false
      bSeparateFillAlpha=false
      outlineMaterial=null
      outlineSize=0
      outlineColor=<rgba 0 0 0 1>
     }
    }
    renderTransform {angle=0, scale=<1 1>, shear=<-12 0>, translation=<0 0>}
    shadowOffset=<0 0>
   }
   slot {class=VerticalBoxSlot, properties {horizontalAlignment=HAlign_Left}}
  }
  Text_CollectionName {
   widgetClassPath=&CommonTextBlock
   parent=VerticalBox_0
   widget=Text_CollectionName
   properties {
    text=Fish
    textTransformPolicy=ToUpper
    font {
     bForceMonospaced=false
     bMaterialIsStencil=false
     fontMaterial=null
     fontObject=&Burbank
     letterSpacing=0
     monospacedWidth=1
     size=24
     skewAmount=0
     typefaceFontName=BurbankRegular-Black
     outlineSettings {
      bApplyOutlineToDropShadows=false
      bMiteredCorners=false
      bSeparateFillAlpha=false
      outlineMaterial=null
      outlineSize=0
      outlineColor=<rgba 0 0 0 1>
     }
    }
    renderTransform {angle=0, scale=<1 1>, shear=<-12 0>, translation=<0 0>}
    shadowOffset=<0 0>
   }
   slot {class=VerticalBoxSlot, properties {horizontalAlignment=HAlign_Left}}
  }
  PlayerProgressContainer {widgetClassPath=/Script/FortniteUI.AthenaCollectionPlayerProgressContainer, bInherited=true, widget=null}
  InfoOverlay {widgetClassPath=/Script/FortniteUI.AthenaCollectionScreenInfoOverlay, bInherited=true, widget=null}
 ]
}
