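# Widget Blueprint for the fishing collection leaderboard, a visual layout subclass of the native AthenaCollectionLeaderboard with an empty event graph.
# A header row shows an uppercase, sheared "Scores" label between two cyan skewed divider bars, the left one mirrored through a negative X scale.
# Below it, the inherited scroll box and dynamic entry box stack fishing-specific leaderboard entry widgets vertically, leaving any population logic outside this Blueprint.
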WidgetBlueprint /FortniteCoreUI/Athena/Collections/AthenaCollectionLeaderboardFishing
graphs [EventGraph {links [], nodes []}]
variables []
widgets {
 info {inheritedWidgetCount=2, parentClass=/Script/FortniteUI.AthenaCollectionLeaderboard, rootWidgetClass=&VerticalBox, widgetCount=7}
 widgets [
  VerticalBox_leaderboard {widgetClassPath=&VerticalBox, widget=VerticalBox_leaderboard}
  HorizontalBox_1 {
   widgetClassPath=&HorizontalBox
   parent=VerticalBox_leaderboard
   widget=HorizontalBox_1
   slot {class=VerticalBoxSlot, properties {padding=<ltrb 0 0 0 8>}}
  }
  Image_5 {
   widgetClassPath=&Image
   parent=HorizontalBox_1
   widget=Image_5
   properties {
    visibility=SelfHitTestInvisible
    brush {
     bIsDynamicallyLoaded=false
     drawAs=Box
     imageType=NoImage
     mirroring=NoMirror
     resourceName=None
     resourceObject=/Game/Athena/UI/ReplayBrowser/M_Ui_Skewed_Vfor9slices_XS
     tiling=NoTile
     imageSize=<6 6>
     margin=<ltrb 0.5 0.5 0.5 0.5>
     outlineSettings {
      bUseBrushTransparency=false
      roundingType=HalfHeightRadius
      width=0
      color {colorUseRule=UseColor_Specified, specifiedColor=<rgba 0 0 0 0>}
      cornerRadii=<0 0 0 0>
     }
     tintColor {colorUseRule=UseColor_Specified, specifiedColor=<rgba 1 1 1 1>}
     uVRegion {bIsValid=false, max=<0 0>, min=<0 0>}
    }
    colorAndOpacity=<rgba 0.0314 0.8549 1 1>
    renderTransform {angle=0, scale=<-1 1>, shear=<0 0>, translation=<0 0>}
   }
   slot {class=HorizontalBoxSlot, properties {verticalAlignment=VAlign_Center, size {sizeRule=Fill, value=1}}}
  }
  ScoresText_1 {
   widgetClassPath=&CommonTextBlock
   parent=HorizontalBox_1
   widget=ScoresText_1
   properties {
    justification=Center
    text=Scores
    textTransformPolicy=ToUpper
    colorAndOpacity {colorUseRule=UseColor_Specified, specifiedColor=<rgba 0.0331 0.855 1 1>}
    font {
     bForceMonospaced=false
     bMaterialIsStencil=false
     fontMaterial=null
     fontObject=&BurbankSmall
     letterSpacing=0
     monospacedWidth=1
     size=16
     skewAmount=0
     typefaceFontName=Black
     outlineSettings {
      bApplyOutlineToDropShadows=false
      bMiteredCorners=false
      bSeparateFillAlpha=false
      outlineMaterial=null
      outlineSize=0
      outlineColor=<rgba 0 0 0 1>
     }
    }
    renderTransform {angle=0, scale=<1 1>, shear=<-12 0>, translation=<0 0>}
    shadowOffset=<0 0>
   }
   slot {class=HorizontalBoxSlot, properties {padding=<ltrb 12 0 12 0>}}
  }
  Image_6 {
   widgetClassPath=&Image
   parent=HorizontalBox_1
   widget=Image_6
   properties {
    visibility=SelfHitTestInvisible
    brush {
     bIsDynamicallyLoaded=false
     drawAs=Box
     imageType=NoImage
     mirroring=NoMirror
     resourceName=None
     resourceObject=&M_Ui_Skewed_Vfor9slices_XS
     tiling=NoTile
     imageSize=<6 6>
     margin=<ltrb 0.5 0.5 0.5 0.5>
     outlineSettings {
      bUseBrushTransparency=false
      roundingType=HalfHeightRadius
      width=0
      color {colorUseRule=UseColor_Specified, specifiedColor=<rgba 0 0 0 0>}
      cornerRadii=<0 0 0 0>
     }
     tintColor {colorUseRule=UseColor_Specified, specifiedColor=<rgba 1 1 1 1>}
     uVRegion {bIsValid=false, max=<0 0>, min=<0 0>}
    }
    colorAndOpacity=<rgba 0.0314 0.8549 1 1>
   }
   slot {class=HorizontalBoxSlot, properties {verticalAlignment=VAlign_Center, size {sizeRule=Fill, value=1}}}
  }
  ScrollBox_Leaderboard {
   widgetClassPath=&ScrollBox
   bInherited=true
   parent=VerticalBox_leaderboard
   widget=ScrollBox_Leaderboard
   slot {class=VerticalBoxSlot, properties {size {sizeRule=Fill, value=1}}}
  }
  EntryBox_Leaderboard {
   widgetClassPath=/Script/FortniteUI.FortDynamicEntryBox
   bInherited=true
   parent=ScrollBox_Leaderboard
   widget=EntryBox_Leaderboard
   properties {
    entryBoxType=Vertical
    entryWidgetClass=/Game/Athena/Collections/BP_AthenaCollectionLeaderboardEntryFishing.BP_AthenaCollectionLeaderboardEntryFishing_C
    numDesignerPreviewEntries=12
    entrySpacing=<0 4>
   }
   slot {class=ScrollBoxSlot}
  }
 ]
}
