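# M_UI_Sidebar_Tags_LinearGradient is a translucent UI material that draws a two-color linear gradient, named for tags in the sidebar.
# MF_UI_LinearGradient builds the blend mask from scalar parameters for start, end, rotation angle, rotation center and offset.
# That mask drives separate lerps between the RGB and alpha of the Color1 and Color2 parameters, so color and transparency can both vary along the gradient.
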Material /FortniteCoreUI/UI/Sidebar/Tags/Materials/M_UI_Sidebar_Tags_LinearGradient
graph {
 properties {
  bUsedWithStaticMesh=false
  blendMode=BLEND_Translucent
  materialDomain=MD_UI
  pixelDepthOffsetMode=PDOM_Legacy
  usedShadingModels=Default Lit
 }
 nodes [
  n1 = VectorParameter() {parameterName=Color1, defaultValue=<rgba 0 0.0091 0.0908 1>}
  n2 = VectorParameter() {parameterName=Color2, defaultValue=<rgba 0.5395 0.227 0.016 1>}
  n3 = MaterialFunctionCall(
   GradientStart=ScalarParameter() {defaultValue=0.5863, parameterName=GradientStart}
   GradientEnd=ScalarParameter() {defaultValue=1, parameterName=GradientEnd}
   RotationAngle=ScalarParameter() {parameterName=RotationAngle}
   RotationCenter=ScalarParameter() {parameterName=RotationCenter}
   GradientOffset=ScalarParameter() {parameterName=GradientOffset}
  ) {materialFunction=&MF_UI_LinearGradient}
  LinearInterpolate(A=@n1.RGB, B=@n2.RGB, Alpha=@n3.Result)
  LinearInterpolate(A=@n1.A, B=@n2.A, Alpha=@n3.Result)
 ]
}
