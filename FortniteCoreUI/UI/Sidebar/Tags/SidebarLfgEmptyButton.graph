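# CommonButtonBase widget for the sidebar LFG empty button: a rounded pill with uppercase ButtonCopy text and a Select input-action icon.
# Hovering plays an animation that turns the fill white, darkens the text and reveals the icon, reversed on unhover.
# Long text shows an ellipsis at rest and scrolls while hovered.
# Icon updates switch its dynamic material to a dark, outline-free look; design-time PreConstruct previews the copy and icon.
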WidgetBlueprint /FortniteCoreUI/UI/Sidebar/Tags/SidebarLfgEmptyButton
animations [Hover {
  binding_count=3
  length_seconds=0.2
  object=&self:Hover
  sequence {bindings [
    InputActionWidget {
     id=#1
     possessed_class=CommonActionWidget
     tracks [{
       type=MovieSceneFloatTrack
       display_name=Render Opacity
       sections [{type=MovieSceneFloatSection, end_frame=4, start_frame=0, channels [None {type=MovieSceneScriptingFloatChannel, keys [[0,0],[1,0],[2,1]]}]}]
      }]
    }
    TextColorBorder {
     id=#2
     possessed_class=CommonBorder
     tracks [{
       type=MovieSceneColorTrack
       display_name=Content Color and Opacity
       sections [{
         type=MovieSceneColorSection
         end_frame=4
         start_frame=0
         channels [
          Color.R {type=MovieSceneScriptingFloatChannel, keys [[0,1],[4,0.0513]]}
          Color.G {type=MovieSceneScriptingFloatChannel, keys [[0,1],[4,0.0513]]}
          Color.B {type=MovieSceneScriptingFloatChannel, keys [[0,1],[4,0.0513]]}
          Color.A {type=MovieSceneScriptingFloatChannel, keys [[0,0.5],[4,1]]}
         ]
        }]
      }]
    }
    EntryBorder {
     id=#3
     possessed_class=CommonBorder
     tracks [{
       type=MovieSceneColorTrack
       display_name=Brush Color
       sections [{
         type=MovieSceneColorSection
         end_frame=4
         start_frame=0
         channels [
          Color.R {type=MovieSceneScriptingFloatChannel, keys [[0,0.004],[4,1]]}
          Color.G {type=MovieSceneScriptingFloatChannel, keys [[0,0.063],[4,1]]}
          Color.B {type=MovieSceneScriptingFloatChannel, keys [[0,0.5029],[4,1]]}
          Color.A {type=MovieSceneScriptingFloatChannel, const [1,1]}
         ]
        }]
      }]
    }
   ]}
 }]
graphs [EventGraph {nodes [
   "On Input Icon Updated (InputActionWidget)"() {is=title K2Node_ComponentBoundEvent}
   > n1 = GetIconDynamicMaterial!(execute, self:Common Action Widget Object Reference=Get InputActionWidget() -> (InputActionWidget:Common Action Widget Object Reference)) -> (then, ReturnValue:Material Instance Dynamic Object Reference)
   > SetScalarParameterValue(self=@n1, ParameterName=Use Dark, Value=1)
   > SetScalarParameterValue(self=@n1, ParameterName=Outline Opacity, Value=0)
   n2 = OnUnhover() {is=title K2Node_CustomEvent}
   n3 = Get Hover() -> (Hover:Widget Animation Object Reference)
   PlayAnimationReverse(execute=[@n2.then, @n10.then], InAnimation=@n3)
   n4 = Get ButtonText() -> (ButtonText:Common Text Object Reference)
   > ResetScrollState(self:Common Text Object Reference=@n4)
   > SetScrollingEnabled!(execute, self:Common Text Object Reference=@n4, bInIsScrollingEnabled:bool=false) -> (then)
   > Set TextOverflowPolicy!(execute, TextOverflowPolicy:ETextOverflowPolicy Enum=Ellipsis, self:Text Object Reference=@n4) -> (then, Output_Get:ETextOverflowPolicy Enum=Clip)
   n5 = OnHover() {is=title K2Node_CustomEvent}
   PlayAnimationForward(execute=[@n5.then, @n9.then], InAnimation=@n3)
   > Set TextOverflowPolicy(TextOverflowPolicy=Clip, self=@n4) -> (Output_Get=Clip)
   > SetScrollingEnabled(self=@n4, bInIsScrollingEnabled=true)
   Event Construct()
   n6 = Get ButtonText()
   n7 = Get ButtonCopy() -> (ButtonCopy:Text)
   > "SetText (Text)"(self=@n6, InText=@n7)
   n8 = Event PreConstruct()
   > Branch(Condition=@n8.IsDesignTime)
   > "SetText (Text)"(self=@n6, InText=@n7)
   > Set RenderOpacity!(execute, RenderOpacity:float=1, self:Widget Object Reference=Get InputActionWidget()) -> (then, Output_Get:float)
   n9 = Event On Hovered()
   n10 = Event On Unhovered()
  ]}]
variables [ButtonCopy {type=string, default=BUTTON TEXT}]
widgets {
 info {inheritedWidgetCount=1, parentClass=&CommonButtonBase, rootWidgetClass=&CommonBorder, widgetCount=10}
 widgets [
  EntryBorder {
   widgetClassPath=&CommonBorder
   widget=EntryBorder
   properties {
    background {
     bIsDynamicallyLoaded=false
     drawAs=Box
     imageType=NoImage
     mirroring=NoMirror
     resourceName=None
     resourceObject=/Game/UI/Sidebar/Materials/M-Button-RoundCorner_SidebarInteractionRound
     tiling=NoTile
     imageSize=<112 112>
     margin=<ltrb 0.5 0.5 0.5 0.5>
     outlineSettings {
      bUseBrushTransparency=false
      roundingType=HalfHeightRadius
      width=0
      color {colorUseRule=UseColor_Specified, specifiedColor=<rgba 0 0 0 0>}
      cornerRadii=<0 0 0 0>
     }
     tintColor {colorUseRule=UseColor_Specified, specifiedColor=<rgba 1 1 1 1>}
     uVRegion {bIsValid=false, max=<0 0>, min=<0 0>}
    }
    brushColor=<rgba 0.004 0.063 0.5029 1>
    padding=<ltrb 50 20 50 20>
   }
  }
  ContentContainer_Padding {
   widgetClassPath=&HorizontalBox
   parent=EntryBorder
   widget=ContentContainer_Padding
   properties {flowDirectionPreference=Culture}
   slot {class=BorderSlot, properties {padding=<ltrb 50 20 50 20>}}
  }
  PaddingForText2x {
   widgetClassPath=&Spacer
   parent=ContentContainer_Padding
   widget=PaddingForText2x
   properties {size=<48 1>}
   slot {class=HorizontalBoxSlot}
  }
  MinMaxWidth {
   widgetClassPath=&SizeBox
   parent=ContentContainer_Padding
   widget=MinMaxWidth
   properties {
    bOverride_MaxDesiredWidth=true
    bOverride_MinDesiredHeight=true
    bOverride_MinDesiredWidth=true
    maxDesiredWidth=400
    minDesiredHeight=72
    minDesiredWidth=250
   }
   slot {class=HorizontalBoxSlot, properties {verticalAlignment=VAlign_Center, size {sizeRule=Fill, value=1}}}
  }
  TextColorBorder {
   widgetClassPath=&CommonBorder
   parent=MinMaxWidth
   widget=TextColorBorder
   properties {horizontalAlignment=HAlign_Center, verticalAlignment=VAlign_Center, contentColorAndOpacity=<rgba 1 1 1 0.5>, padding=<ltrb 0 0 0 0>}
   slot {class=SizeBoxSlot}
  }
  ButtonText {
   widgetClassPath=&CommonTextBlock
   bIsVariable=true
   parent=TextColorBorder
   widget=ButtonText
   properties {
    clipping=ClipToBounds
    justification=Center
    scrollStyle=/FortUILibrary/Styles/Typography/Foundation/TextScrollStyle-NoFade.TextScrollStyle-NoFade_C
    text=Text Block
    textOverflowPolicy=Ellipsis
    textTransformPolicy=ToUpper
    font {
     bForceMonospaced=false
     bMaterialIsStencil=false
     fontMaterial=null
     fontObject=&BurbankSmall
     letterSpacing=0
     monospacedWidth=1
     size=18
     skewAmount=0
     typefaceFontName=Black
     outlineSettings {
      bApplyOutlineToDropShadows=false
      bMiteredCorners=false
      bSeparateFillAlpha=false
      outlineMaterial=null
      outlineSize=0
      outlineColor=<rgba 0 0 0 1>
     }
    }
    shadowOffset=<0 0>
   }
   slot {class=BorderSlot, properties {horizontalAlignment=HAlign_Center, verticalAlignment=VAlign_Center, padding=<ltrb 0 0 0 0>}}
  }
  PaddingForText {
   widgetClassPath=&Spacer
   parent=ContentContainer_Padding
   widget=PaddingForText
   properties {size=<16 1>}
   slot {class=HorizontalBoxSlot}
  }
  ForceSpaceContainer {
   widgetClassPath=&Overlay
   parent=ContentContainer_Padding
   widget=ForceSpaceContainer
   slot {class=HorizontalBoxSlot, properties {horizontalAlignment=HAlign_Center, verticalAlignment=VAlign_Center}}
  }
  InputActionWidget {
   widgetClassPath=&CommonActionWidget
   bInherited=true
   bIsVariable=true
   parent=ForceSpaceContainer
   widget=InputActionWidget
   properties {renderOpacity=0, inputActions [{dataTable=&InputActionDataTable, rowName=Select}]}
   slot {class=OverlaySlot}
  }
  ForcedSpace32x32 {widgetClassPath=&Spacer, parent=ForceSpaceContainer, widget=ForcedSpace32x32, properties {size=<32 32>}, slot {class=OverlaySlot}}
 ]
}
