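# Sidebar tag widget that shows an LFG (looking-for-group) state icon as a single tintable SDF material image.
# SetMaterial applies a soft material to LFG_Icon and caches its dynamic instance in LFG_DynamicMaterial, while SetColor tints the image.
# PreConstruct calls both functions for design-time preview.
# At runtime, OnInitialized applies only LFG_Color, leaving the brush's default material unless SetMaterial is called separately.
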WidgetBlueprint /FortniteCoreUI/UI/Sidebar/Tags/SidebarLfgState_Image
graphs [
 EventGraph {nodes [
   n1 = Event PreConstruct()
   > Branch(Condition=@n1.IsDesignTime)
   > SetMaterial!(
    execute
    self:Self Object Reference
    SoftMaterial:Material Interface Soft Object Reference=Get LFG_Material() -> (LFG_Material:Material Interface Soft Object Reference)
   ) -> (then)
   > SetColor!(execute, self, InColorAndOpacity:Linear Color Structure=Get LFG_Color() -> (LFG_Color:Linear Color Structure)) -> (then)
   Event OnInitialized()
   > SetColor(InColorAndOpacity=Get LFG_Color())
  ]}
 SetColor {nodes [
   n1 = SetColor!() -> (then, InColorAndOpacity:Linear Color Structure) {is=title K2Node_FunctionEntry}
   > SetColorAndOpacity(self:Image Object Reference=Get LFG_Icon() -> (LFG_Icon:Image Object Reference), InColorAndOpacity:Linear Color Structure=@n1)
  ]}
 SetMaterial {nodes [
   n1 = SetMaterial!() -> (then, SoftMaterial:Material Interface Soft Object Reference) {is=title K2Node_FunctionEntry}
   n2 = Get LFG_Icon()
   > SetBrushFromSoftMaterial!(execute, self:Image Object Reference=@n2, SoftMaterial:Material Interface Soft Object Reference=@n1) -> (then)
   > n3 = GetDynamicMaterial(self=@n2)
   > Set LFG_DynamicMaterial(LFG_DynamicMaterial:Material Instance Dynamic Object Reference=@n3) -> (Output_Get:Material Instance Dynamic Object Reference)
  ]}
]
variables [
 LFG_Color {type=LinearColor, default=<rgba 0.0331 0.402 1 1>}
 LFG_DynamicMaterial {type=MaterialInstanceDynamic}
 LFG_Material {type=MaterialInterface, default=/Game/UI/Sidebar/Tags/Materials/MI_UI_LfgStateIconSDF}
]
widgets {
 info {inheritedWidgetCount=0, parentClass=&UserWidget, rootWidgetClass=&Image, widgetCount=1}
 widgets [LFG_Icon {
   widgetClassPath=&Image
   bIsVariable=true
   widget=LFG_Icon
   properties {
    visibility=SelfHitTestInvisible
    brush {
     bIsDynamicallyLoaded=false
     drawAs=Image
     imageType=NoImage
     mirroring=NoMirror
     resourceName=None
     resourceObject=&MI_UI_LfgStateIconSDF
     tiling=NoTile
     imageSize=<200 200>
     margin=<ltrb 0 0 0 0>
     outlineSettings {
      bUseBrushTransparency=false
      roundingType=HalfHeightRadius
      width=0
      color {colorUseRule=UseColor_Specified, specifiedColor=<rgba 0 0 0 0>}
      cornerRadii=<0 0 0 0>
     }
     tintColor {colorUseRule=UseColor_Specified, specifiedColor=<rgba 1 1 1 1>}
     uVRegion {bIsValid=false, max=<0 0>, min=<0 0>}
    }
   }
  }]
}
