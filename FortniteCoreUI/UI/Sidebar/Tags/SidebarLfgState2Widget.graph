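# Sidebar widget that shows the current looking-for-group (LFG) social tag state as a centered icon, message and button.
# On state change it updates icon and text visibility, message and button label, and pads the button on mobile.
# GetIcon feeds a per-state texture into an SDF dynamic material, except Shuffling, which swaps in its own material.
# Shuffling also makes the whole widget HitTestInvisible, blocking button input.
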WidgetBlueprint /FortniteCoreUI/UI/Sidebar/Tags/SidebarLfgState2Widget
graphs [
 EventGraph {nodes [
   Event OnInitialized()
   > Branch(Condition=IsMobileGame())
   > SetPadding!(
    execute
    self:Vertical Box Slot Object Reference=SlotAsVerticalBoxSlot(Widget=Get Button() -> (Button:Common Button Base Object Reference))
    InPadding_Left:float=0
    InPadding_Top:float=15
    InPadding_Right:float=0
    InPadding_Bottom:float=15
   ) -> (then)
   n1 = Event OnStateChanged!() -> (
    OutputDelegate
    then
    NewLfgState:ELfgState Enum=NoResults
    BodyVisibility:ESlateVisibility Enum=Visible
    StateText:"Text (by ref)"
    StateButton:"Text (by ref)"
   ) {is=title K2Node_Event}
   > UpdateButtonVisibility!(execute, self, LfgState:ELfgState Enum=@n1.NewLfgState) -> (then)
   > GetIcon!(execute, self, LfgState:ELfgState Enum=@n1.NewLfgState, Icon Visiblity:ESlateVisibility Enum=@n1.BodyVisibility) -> (then)
   n2 = Get SidebarLfgState_RichText() -> (SidebarLfgState_RichText:Sidebar Lfg State Rich Text Object Reference)
   > SetVisibility(self=@n2, InVisibility=@n1.BodyVisibility)
   > SetText(
    self=Get RText_LFGText!(self:Sidebar Lfg State Rich Text Object Reference=@n2) -> (RText_LFGText:Common Rich Text Block Object Reference)
    InText=@n1.StateText
   )
   > n3 = Cast To SidebarLfgEmptyButton!(execute, Object=Get Button()) -> (then, CastFailed, AsSidebar Lfg Empty Button:Sidebar Lfg Empty Button Object Reference) {is=title K2Node_DynamicCast}
   > "SetText (Text)"(
    self=Get ButtonText!(self:Sidebar Lfg Empty Button Object Reference=@n3) -> (ButtonText:Common Text Object Reference)
    InText=@n1.StateButton
   )
  ]}
 GetIcon {nodes [
   n1 = GetIcon!() -> (then, LfgState:ELfgState Enum=NoResults, Icon Visiblity:ESlateVisibility Enum=Visible) {is=title K2Node_FunctionEntry}
   > SetVisibility(self=Get SidebarLfgState_Image() -> (SidebarLfgState_Image:Sidebar Lfg State Image Object Reference), InVisibility=@n1.Icon Visiblity)
   > n2 = Branch(Condition="Not Equal (Enum)"!(A:ELfgState Enum=@n1, B:ELfgState Enum=Shuffling) -> (ReturnValue:bool))
   n3 = Get SidebarLfgState_Image()
   > SetMaterial!(
    execute
    self:Sidebar Lfg State Image Object Reference=@n3
    SoftMaterial:Material Interface Soft Object Reference=/Game/UI/Sidebar/Tags/Materials/MI_UI_LfgStateIconSDF
   ) -> (then)
   > SetTextureParameterValue(
    self=Get LFG_DynamicMaterial!(self:Sidebar Lfg State Image Object Reference=@n3) -> (LFG_DynamicMaterial:Material Instance Dynamic Object Reference)
    ParameterName=Icon SDF
    Value=Select!(
     NoResults:Texture Object Reference=/Game/UI/Sidebar/Tags/Textures/T_UI_Tags_Exclamation_SDF
     Shuffling:Texture Object Reference=/Game/UI/Sidebar/Tags/Textures/T_UI_Tags_Shuffle
     NoSocialTagsAdded:Texture Object Reference=/Game/UI/Sidebar/Tags/Textures/T_UI_Tags_Add_SDF
     LfgOff:Texture Object Reference=/Game/UI/Sidebar/Tags/Textures/T_UI_Tags_Search_SDF
     LfgUnavailable:Texture Object Reference=/Game/UI/Sidebar/Tags/Textures/T_UI_Tags_Unavailable_SDF
     LfgOffNoSocialTags:Texture Object Reference=&T_UI_Tags_Search_SDF
     NoSocialTagsSelected:Texture Object Reference=/Game/UI/Sidebar/Tags/Textures/T_UI_Tags_Checkmark_SDF
     NoFriendsSuggestLFG:Texture Object Reference=&T_UI_Tags_Add_SDF
     RebootInvitePlayerButton:Texture Object Reference
     RebootInviteAllRallied:Texture Object Reference
     ChangeTagsButton:Texture Object Reference
     ApplyButton:Texture Object Reference
     ShuffleButton:Texture Object Reference
     LfgOffSocialBan:Texture Object Reference
     Index:ELfgState Enum=@n1
    ) -> (ReturnValue:Texture Object Reference)
   )
   SetMaterial(execute=@n2.else, self=Get SidebarLfgState_Image(), SoftMaterial=/Game/UI/Sidebar/Tags/Materials/M_UI_Tags_Shuffle_Inst)
  ]}
 UpdateButtonVisibility {nodes [
   n1 = UpdateButtonVisibility!() -> (then, LfgState:ELfgState Enum=NoResults) {is=title K2Node_FunctionEntry}
   > n2 = Branch(Condition="Equal (Enum)"(A:ELfgState Enum=@n1, B:ELfgState Enum=Shuffling))
   > SetVisibility(InVisibility=HitTestInvisible)
   SetVisibility(execute=@n2.else, InVisibility=SelfHitTestInvisible)
  ]}
]
variables [LfgStateTextureArray {
  type="array<Texture2D>"
  default [
   &T_UI_Tags_Exclamation_SDF
   &T_UI_Tags_Shuffle
   &T_UI_Tags_Add_SDF
   &T_UI_Tags_Search_SDF
   &T_UI_Tags_Unavailable_SDF
   &T_UI_Tags_Search_SDF
   &T_UI_Tags_Checkmark_SDF
  ]
 }]
widgets {
 info {inheritedWidgetCount=1, parentClass=/Script/FortniteUI.FortSidebarLfgState2Widget, rootWidgetClass=&Overlay, widgetCount=5}
 widgets [
  Overlay_0 {widgetClassPath=&Overlay, widget=Overlay_0}
  VerticalBox_0 {
   widgetClassPath=&VerticalBox
   parent=Overlay_0
   widget=VerticalBox_0
   slot {class=OverlaySlot, properties {horizontalAlignment=HAlign_Fill, verticalAlignment=VAlign_Center}}
  }
  SidebarLfgState_Image {
   widgetClassPath=&/SidebarLfgState_Image
   bIsVariable=true
   parent=VerticalBox_0
   widget=SidebarLfgState_Image
   slot {class=VerticalBoxSlot, properties {horizontalAlignment=HAlign_Center, padding=<ltrb 0 0 0 24>}}
  }
  SidebarLfgState_RichText {
   widgetClassPath=&/SidebarLfgState_RichText
   bIsVariable=true
   parent=VerticalBox_0
   widget=SidebarLfgState_RichText
   slot {class=VerticalBoxSlot, properties {padding=<ltrb 32 0 32 24>}}
  }
  Button {
   widgetClassPath=&/SidebarLfgEmptyButton
   bInherited=true
   bIsVariable=true
   parent=VerticalBox_0
   widget=Button
   slot {class=VerticalBoxSlot, properties {horizontalAlignment=HAlign_Center}}
  }
 ]
}
