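# Sidebar checkbox widget built on native FortSidebarCheckboxWidget, drawn as a button holding a checked/unchecked icon switcher.
# A button click or the Toggle Checkbox event inverts bIsSelected through SetSelected, broadcasts OnClicked, then calls HandleCheckboxClicked.
# Refresh picks the switcher image from bIsSelected, running on Construct and on every SetSelected call.
# SetSelected itself updates state without broadcasting, and the checkbox defaults to selected.
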WidgetBlueprint /FortniteCoreUI/UI/Sidebar/Tags/FortSidebarCheckbox
graphs [
 EventGraph {nodes [
   n1 = "On Clicked (CheckButton)"() {is=title K2Node_ComponentBoundEvent}
   ToggleCheckbox(execute=[@n1.then, @n2.then])
   > Call On Clicked!(execute:Exec, self:Fort Sidebar Checkbox Object Reference) -> (then:Exec) {is=title K2Node_CallDelegate}
   > HandleCheckboxClicked(self:Fort Sidebar Checkbox Widget Object Reference)
   Event Construct()
   > Refresh()
   n2 = Event Toggle Checkbox() {is=title K2Node_Event}
   Event PreConstruct()
   Event Tick()
  ]}
 OnClicked {nodes [OnClicked() {is=title K2Node_FunctionEntry}]}
 Refresh {nodes [
   Refresh() {is=title K2Node_FunctionEntry}
   > SetActiveWidget(
    self:Widget Switcher Object Reference=Get Switcher_CheckBox() -> (Switcher_CheckBox:Widget Switcher Object Reference)
    Widget=Select(
     Option 0:Widget Object Reference=Get Image_Unchecked() -> (Image_Unchecked:Image Object Reference)
     Option 1:Widget Object Reference=Get Image_Checked() -> (Image_Checked:Image Object Reference)
     Index=Get bIsSelected() -> (bIsSelected:bool)
    ) -> (ReturnValue:Widget Object Reference)
   )
  ]}
 SetSelected {nodes [
   n1 = SetSelected!() -> (then, NewSelected:bool=~) {is=title K2Node_FunctionEntry}
   > Set bIsSelected(bIsSelected:bool=@n1) -> (Output_Get:bool)
   > Refresh() {is=title K2Node_CallFunction}
  ]}
 ToggleCheckbox {nodes [
   ToggleCheckbox() {is=title K2Node_FunctionEntry}
   > SetSelected!(execute, self, NewSelected:bool=NOT Boolean(A=Get bIsSelected())) -> (then) {is=title K2Node_CallFunction}
  ]}
]
variables [
 bIsSelected {type=boolean, default=true}
 OnClicked {}
]
widgets {
 info {inheritedWidgetCount=2, parentClass=/Script/FortniteUI.FortSidebarCheckboxWidget, rootWidgetClass=&Button, widgetCount=4}
 widgets [
  CheckButton {
   widgetClassPath=&Button
   bInherited=true
   widget=CheckButton
   properties {widgetStyle {
     clickedSlateSound {resourceObject=null}
     disabled {
      bIsDynamicallyLoaded=false
      drawAs=NoDrawType
      imageType=NoImage
      mirroring=NoMirror
      resourceName=None
      resourceObject=null
      tiling=NoTile
      imageSize=<0 0>
      margin=<ltrb 0 0 0 0>
      outlineSettings {
       bUseBrushTransparency=false
       roundingType=HalfHeightRadius
       width=0
       color {colorUseRule=UseColor_Specified, specifiedColor=<rgba 0 0 0 0>}
       cornerRadii=<0 0 0 0>
      }
      tintColor {colorUseRule=UseColor_Specified, specifiedColor=<rgba 1 1 1 1>}
      uVRegion {bIsValid=false, max=<0 0>, min=<0 0>}
     }
     disabledForeground {colorUseRule=UseColor_Specified, specifiedColor=<rgba 0.5271 0.5271 0.5271 1>}
     hovered {
      bIsDynamicallyLoaded=false
      drawAs=RoundedBox
      imageType=NoImage
      mirroring=NoMirror
      resourceName=None
      resourceObject=null
      tiling=NoTile
      imageSize=<32 32>
      margin=<ltrb 0 0 0 0>
      outlineSettings {
       bUseBrushTransparency=true
       roundingType=FixedRadius
       width=1
       color {colorUseRule=UseColor_Specified, specifiedColor=<rgba 0.7243 0.7243 0.7243 1>}
       cornerRadii=<4 4 4 4>
      }
      tintColor {colorUseRule=UseColor_Specified, specifiedColor=<rgba 0.7243 0.7243 0.7243 1>}
      uVRegion {bIsValid=false, max=<0 0>, min=<0 0>}
     }
     hoveredForeground {colorUseRule=UseColor_Specified, specifiedColor=<rgba 0 1 1 1>}
     hoveredSlateSound {resourceObject=null}
     normal {
      bIsDynamicallyLoaded=false
      drawAs=NoDrawType
      imageType=NoImage
      mirroring=NoMirror
      resourceName=None
      resourceObject=null
      tiling=NoTile
      imageSize=<32 32>
      margin=<ltrb 0 0 0 0>
      outlineSettings {
       bUseBrushTransparency=true
       roundingType=FixedRadius
       width=1
       color {colorUseRule=UseColor_Specified, specifiedColor=<rgba 0.6951 0.6951 0.6951 1>}
       cornerRadii=<4 4 4 4>
      }
      tintColor {colorUseRule=UseColor_Specified, specifiedColor=<rgba 0.4955 0.4955 0.4955 1>}
      uVRegion {bIsValid=false, max=<0 0>, min=<0 0>}
     }
     normalForeground {colorUseRule=UseColor_Specified, specifiedColor=<rgba 0 1 1 1>}
     normalPadding=<ltrb 12 1.5 12 1.5>
     pressed {
      bIsDynamicallyLoaded=false
      drawAs=RoundedBox
      imageType=NoImage
      mirroring=NoMirror
      resourceName=None
      resourceObject=null
      tiling=NoTile
      imageSize=<32 32>
      margin=<ltrb 0 0 0 0>
      outlineSettings {
       bUseBrushTransparency=true
       roundingType=FixedRadius
       width=1
       color {colorUseRule=UseColor_Specified, specifiedColor=<rgba 0.7243 0.7243 0.7243 1>}
       cornerRadii=<4 4 4 4>
      }
      tintColor {colorUseRule=UseColor_Specified, specifiedColor=<rgba 0.3843 0.3843 0.3843 1>}
      uVRegion {bIsValid=false, max=<0 0>, min=<0 0>}
     }
     pressedForeground {colorUseRule=UseColor_Specified, specifiedColor=<rgba 0 1 1 1>}
     pressedPadding=<ltrb 12 1.5 12 1.5>
     pressedSlateSound {resourceObject=null}
    }}
  }
  Switcher_CheckBox {
   widgetClassPath=&WidgetSwitcher
   bInherited=true
   bIsVariable=true
   parent=CheckButton
   widget=Switcher_CheckBox
   slot {class=ButtonSlot}
  }
  Image_Checked {
   widgetClassPath=&Image
   bIsVariable=true
   parent=Switcher_CheckBox
   widget=Image_Checked
   properties {brush {
     bIsDynamicallyLoaded=false
     drawAs=Image
     imageType=NoImage
     mirroring=NoMirror
     resourceName=None
     resourceObject=/Game/UI/Foundation/Textures/Icons/Checks/T-Icon-CheckedBox-76
     tiling=NoTile
     imageSize=<76 76>
     margin=<ltrb 0 0 0 0>
     outlineSettings {
      bUseBrushTransparency=false
      roundingType=HalfHeightRadius
      width=0
      color {colorUseRule=UseColor_Specified, specifiedColor=<rgba 0 0 0 0>}
      cornerRadii=<0 0 0 0>
     }
     tintColor {colorUseRule=UseColor_Specified, specifiedColor=<rgba 1 1 1 1>}
     uVRegion {bIsValid=false, max=<0 0>, min=<0 0>}
    }}
   slot {class=WidgetSwitcherSlot}
  }
  Image_Unchecked {
   widgetClassPath=&Image
   bIsVariable=true
   parent=Switcher_CheckBox
   widget=Image_Unchecked
   properties {brush {
     bIsDynamicallyLoaded=false
     drawAs=Image
     imageType=NoImage
     mirroring=NoMirror
     resourceName=None
     resourceObject=/Game/UI/Foundation/Textures/Icons/Checks/T-Icon-Confirmation_Empty-76
     tiling=NoTile
     imageSize=<76 76>
     margin=<ltrb 0 0 0 0>
     outlineSettings {
      bUseBrushTransparency=false
      roundingType=HalfHeightRadius
      width=0
      color {colorUseRule=UseColor_Specified, specifiedColor=<rgba 0 0 0 0>}
      cornerRadii=<0 0 0 0>
     }
     tintColor {colorUseRule=UseColor_Specified, specifiedColor=<rgba 1 1 1 1>}
     uVRegion {bIsValid=false, max=<0 0>, min=<0 0>}
    }}
   slot {class=WidgetSwitcherSlot}
  }
 ]
}
