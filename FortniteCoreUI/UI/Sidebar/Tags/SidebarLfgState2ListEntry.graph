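# Sidebar list entry showing a looking-for-group (LFG) state with an icon, message and action button.
# OnStateChanged fills text and visibility, picking a per-state SDF icon; Shuffling swaps in its own material and makes the entry hit-test-invisible.
# Hover animates button colors and input glyph, via the click-through button or, for non-mouse input, entry hover events.
# Button text scrolls on button hover or touch input; input-method changes adjust button visibility.
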WidgetBlueprint /FortniteCoreUI/UI/Sidebar/Tags/SidebarLfgState2ListEntry
animations [
 Hovered {
  binding_count=3
  display_label=""
  length_seconds=0.2
  object=&self:Hovered
  sequence {bindings [
    InputActionWidget {
     id=#1
     possessed_class=CommonActionWidget
     tracks [{
       type=MovieSceneFloatTrack
       display_name=Render Opacity
       sections [{type=MovieSceneFloatSection, end_frame=4, start_frame=0, channels [None {type=MovieSceneScriptingFloatChannel, keys [[0,0],[1,0],[2,1]]}]}]
      }]
    }
    EntryBorder {
     id=#2
     possessed_class=Border
     tracks [
      {
       type=MovieSceneWidgetMaterialTrack
       sections [{
         type=MovieSceneParameterSection
         end_frame=4
         start_frame=0
         channels [AnimateHighlight {type=MovieSceneScriptingFloatChannel, keys [[0,0],[3,1],[4,1]]}]
        }]
      }
      {
       type=MovieSceneColorTrack
       display_name=Brush Color
       sections [{
         type=MovieSceneColorSection
         end_frame=4
         start_frame=0
         channels [
          Color.R {type=MovieSceneScriptingFloatChannel, keys [[0,0.004],[4,1]]}
          Color.G {type=MovieSceneScriptingFloatChannel, keys [[0,0.063],[4,1]]}
          Color.B {type=MovieSceneScriptingFloatChannel, keys [[0,0.5029],[4,1]]}
          Color.A {type=MovieSceneScriptingFloatChannel, const [1,1]}
         ]
        }]
      }
     ]
    }
    TextColorBorder {
     id=#3
     possessed_class=CommonBorder
     tracks [{
       type=MovieSceneColorTrack
       display_name=Content Color and Opacity
       sections [{
         type=MovieSceneColorSection
         end_frame=4
         start_frame=0
         channels [
          Color.R {type=MovieSceneScriptingFloatChannel, keys [[0,1],[4,0.0513]]}
          Color.G {type=MovieSceneScriptingFloatChannel, keys [[0,1],[4,0.0513]]}
          Color.B {type=MovieSceneScriptingFloatChannel, keys [[0,1],[4,0.0513]]}
          Color.A {type=MovieSceneScriptingFloatChannel, keys [[0,0.5],[4,1]]}
         ]
        }]
      }]
    }
   ]}
 }
 Selected {
  binding_count=0
  display_label=""
  length_seconds=0.2
  object=&self:Selected
  sequence {bindings [], display_rate {numerator=60}}
 }
]
graphs [
 EventGraph {nodes [
   "On Input Icon Updated (InputActionWidget)"() {is=title K2Node_ComponentBoundEvent}
   > n1 = GetIconDynamicMaterial!(execute, self:Common Action Widget Object Reference=Get InputActionWidget() -> (InputActionWidget:Common Action Widget Object Reference)) -> (then, ReturnValue:Material Instance Dynamic Object Reference)
   > SetScalarParameterValue(self=@n1, ParameterName=Use Dark, Value=1)
   > SetScalarParameterValue(self=@n1, ParameterName=Outline Opacity, Value=0)
   n2 = "On Hovered (FortClickThroughButton)"() {is=title K2Node_ComponentBoundEvent}
   PlayAnimation(execute=[@n2.then, @n8.then], InAnimation=Get Hovered() -> (Hovered:Widget Animation Object Reference))
   n3 = Get ButtonText() -> (ButtonText:Common Text Object Reference)
   > Set TextOverflowPolicy!(execute, TextOverflowPolicy:ETextOverflowPolicy Enum=Clip, self:Text Object Reference=@n3) -> (then, Output_Get:ETextOverflowPolicy Enum=Clip)
   > SetScrollingEnabled!(execute, self:Common Text Object Reference=@n3, bInIsScrollingEnabled:bool=true) -> (then)
   n4 = "On Unhovered (FortClickThroughButton)"() {is=title K2Node_ComponentBoundEvent}
   PlayAnimation(execute=[@n11.then, @n4.then], InAnimation=Get Hovered(), PlayMode=Reverse, PlaybackSpeed=2)
   n5 = Get ButtonText()
   > ResetScrollState(self:Common Text Object Reference=@n5)
   > SetScrollingEnabled(self=@n5, bInIsScrollingEnabled=false)
   > Set TextOverflowPolicy(TextOverflowPolicy=Ellipsis, self=@n5) -> (Output_Get=Clip)
   n6 = HandleInputMethodChanged!() -> (OutputDelegate, then, bNewInputType:ECommonInputType Enum) {is=title K2Node_CustomEvent}
   > "On Input Method Changed "!(execute, self, bNewInputType:ECommonInputType Enum=@n6.bNewInputType) -> (then)
   Event OnInitialized()
   > Branch(Condition=IsMobileGame())
   > SetPadding!(
    execute
    self:Overlay Slot Object Reference=SlotAsOverlaySlot(Widget=Get EntryBorder() -> (EntryBorder:Border Object Reference))
    InPadding_Left:float=0
    InPadding_Top:float=15
    InPadding_Right:float=0
    InPadding_Bottom:float=15
   ) -> (then)
   Event On Hovered()
   n7 = "Not Equal (Enum)"!(A:ECommonInputType Enum=GetCurrentInputType(self=Get CommonInputSubsystem()), B:ECommonInputType Enum=MouseAndKeyboard) -> (ReturnValue:bool)
   > n8 = Branch(Condition=@n7)
   n9 = Event OnStateChanged!() -> (
    OutputDelegate
    then
    NewLfgState:ELfgState Enum=NoResults
    BodyVisibility:ESlateVisibility Enum=Visible
    StateText:"Text (by ref)"
    StateButton:"Text (by ref)"
   ) {is=title K2Node_Event}
   > SetVisibility(self=Get SidebarLfgState_Image() -> (SidebarLfgState_Image:Sidebar Lfg State Image Object Reference), InVisibility=@n9.BodyVisibility)
   > UpdateButtonVisibility!(execute, self, LfgState:ELfgState Enum=@n9.NewLfgState) -> (then)
   > GetIcon!(execute, self, LfgState:ELfgState Enum=@n9.NewLfgState) -> (then)
   n10 = Get SidebarLfgState_RichText() -> (SidebarLfgState_RichText:Sidebar Lfg State Rich Text Object Reference)
   > SetVisibility(self=@n10, InVisibility=@n9.BodyVisibility)
   > SetText(
    self=Get RText_LFGText!(self:Sidebar Lfg State Rich Text Object Reference=@n10) -> (RText_LFGText:Common Rich Text Block Object Reference)
    InText=@n9.StateText
   )
   > "SetText (Text)"(self=Get ButtonText(), InText=@n9.StateButton)
   Event On Unhovered()
   > n11 = Branch(Condition=@n7)
   Event Construct()
   n12 = Create Event()
   n13 = Get CommonInputSubsystem()
   > Bind Event to On Input Method Changed(self=@n13, Delegate=@n12)
   > "On Input Method Changed "(bNewInputType=GetCurrentInputType(self=@n13))
   Event Destruct()
   > Unbind Event from On Input Method Changed(self=@n13, Delegate=@n12)
   Branch(Condition=Event PreConstruct().IsDesignTime)
   > GetIcon(LfgState=LfgUnavailable)
   > UpdateButtonVisibility(LfgState=LfgUnavailable)
   n14 = Branch()
   n15 = Get Selected() -> (Selected:Widget Animation Object Reference)
   > PlayAnimation(InAnimation=@n15)
   PlayAnimation(execute=@n14.else, InAnimation=@n15, PlayMode=Reverse, PlaybackSpeed=2)
   > Branch(Condition=IsAnimationPlaying(InAnimation=Get Selected()))
   > PlaySound2D(Sound=/Game/Sounds/Fort_UI_Sounds/FrontEnd/New_FrontEnd/Fortnite_UI_MainMenu_Generic_01_Cue)
   "On Menu Open Changed (MenuAnchor_Actions)"!() -> (OutputDelegate, then, bIsOpen:bool) {is=title K2Node_ComponentBoundEvent}
  ]}
 GetIcon {nodes [
   n1 = GetIcon!() -> (then, LfgState:ELfgState Enum=NoResults) {is=title K2Node_FunctionEntry}
   > n2 = Branch(Condition="Not Equal (Enum)"(A:ELfgState Enum=@n1, B:ELfgState Enum=Shuffling))
   n3 = Get SidebarLfgState_Image()
   > SetMaterial!(
    execute
    self:Sidebar Lfg State Image Object Reference=@n3
    SoftMaterial:Material Interface Soft Object Reference=/Game/UI/Sidebar/Tags/Materials/MI_UI_LfgStateIconSDF
   ) -> (then)
   > SetTextureParameterValue(
    self=Get LFG_DynamicMaterial!(self:Sidebar Lfg State Image Object Reference=@n3) -> (LFG_DynamicMaterial:Material Instance Dynamic Object Reference)
    ParameterName=Icon SDF
    Value=Select!(
     NoResults:Texture Object Reference=/Game/UI/Sidebar/Tags/Textures/T_UI_Tags_Exclamation_SDF
     Shuffling:Texture Object Reference=/Game/UI/Sidebar/Tags/Textures/T_UI_Tags_Shuffle
     NoSocialTagsAdded:Texture Object Reference=/Game/UI/Sidebar/Tags/Textures/T_UI_Tags_Add_SDF
     LfgOff:Texture Object Reference=/Game/UI/Sidebar/Tags/Textures/T_UI_Tags_Search_SDF
     LfgUnavailable:Texture Object Reference=/Game/UI/Sidebar/Tags/Textures/T_UI_Tags_Unavailable_SDF
     LfgOffNoSocialTags:Texture Object Reference=&T_UI_Tags_Search_SDF
     NoSocialTagsSelected:Texture Object Reference=/Game/UI/Sidebar/Tags/Textures/T_UI_Tags_Checkmark_SDF
     NoFriendsSuggestLFG:Texture Object Reference=&T_UI_Tags_Add_SDF
     RebootInvitePlayerButton:Texture Object Reference
     RebootInviteAllRallied:Texture Object Reference
     ChangeTagsButton:Texture Object Reference
     ApplyButton:Texture Object Reference
     ShuffleButton:Texture Object Reference
     LfgOffSocialBan:Texture Object Reference
     Index:ELfgState Enum=@n1
    ) -> (ReturnValue:Texture Object Reference)
   )
   SetMaterial(execute=@n2.else, self=Get SidebarLfgState_Image(), SoftMaterial=/Game/UI/Sidebar/Tags/Materials/M_UI_Tags_Shuffle_Inst)
  ]}
 "On Input Method Changed " {nodes [
   n1 = "On Input Method Changed "!() -> (then, bNewInputType:ECommonInputType Enum) {is=title K2Node_FunctionEntry}
   > SetVisibility(
    self=Get FortClickThroughButton() -> (FortClickThroughButton:Fort Click Through Button Object Reference)
    InVisibility=Select!(
     MouseAndKeyboard:ESlateVisibility Enum=Visible
     Gamepad:ESlateVisibility Enum=SelfHitTestInvisible
     Touch:ESlateVisibility Enum=Visible
     Count:ESlateVisibility Enum=Visible
     Index:ECommonInputType Enum=@n1
    ) -> (ReturnValue:ESlateVisibility Enum)
   )
   > Branch(Condition=Select(MouseAndKeyboard:bool=false, Gamepad:bool=false, Touch:bool=true, Count:bool=false, Index=@n1) -> (ReturnValue:bool))
   n2 = Get ButtonText()
   > Set TextOverflowPolicy(TextOverflowPolicy=Clip, self=@n2) -> (Output_Get=Clip)
   > SetScrollingEnabled(self=@n2, bInIsScrollingEnabled=true)
  ]}
 UpdateButtonVisibility {nodes [
   n1 = UpdateButtonVisibility!() -> (then, LfgState:ELfgState Enum=NoResults) {is=title K2Node_FunctionEntry}
   > n2 = Branch(Condition="Equal (Enum)"(A:ELfgState Enum=@n1, B:ELfgState Enum=Shuffling))
   > SetVisibility(InVisibility=HitTestInvisible)
   SetVisibility(execute=@n2.else, InVisibility=SelfHitTestInvisible)
  ]}
]
variables []
widgets {
 info {inheritedWidgetCount=3, parentClass=/Script/FortniteUI.FortSidebarLfgState2ListEntry, rootWidgetClass=&Overlay, widgetCount=17}
 widgets [
  Overlay_0 {widgetClassPath=&Overlay, widget=Overlay_0}
  VerticalBox_0 {
   widgetClassPath=&VerticalBox
   parent=Overlay_0
   widget=VerticalBox_0
   slot {class=OverlaySlot, properties {horizontalAlignment=HAlign_Fill, verticalAlignment=VAlign_Center}}
  }
  SidebarLfgState_Image {
   widgetClassPath=&/SidebarLfgState_Image
   bIsVariable=true
   parent=VerticalBox_0
   widget=SidebarLfgState_Image
   slot {class=VerticalBoxSlot, properties {horizontalAlignment=HAlign_Center, verticalAlignment=VAlign_Top, padding=<ltrb 0 0 0 24>}}
  }
  SidebarLfgState_RichText {
   widgetClassPath=&/SidebarLfgState_RichText
   bIsVariable=true
   parent=VerticalBox_0
   widget=SidebarLfgState_RichText
   slot {class=VerticalBoxSlot, properties {verticalAlignment=VAlign_Top, padding=<ltrb 32 0 32 24>}}
  }
  ButtonContainer {
   widgetClassPath=&Overlay
   bIsVariable=true
   parent=VerticalBox_0
   widget=ButtonContainer
   slot {class=VerticalBoxSlot, properties {horizontalAlignment=HAlign_Center}}
  }
  MenuAnchor_Actions {
   widgetClassPath=/Script/UMG.MenuAnchor
   bInherited=true
   bIsVariable=true
   parent=ButtonContainer
   widget=MenuAnchor_Actions
   properties {menuClass=/Game/UI/SocialPanel/UserActionMenu.UserActionMenu_C, placement=MenuPlacement_MenuRight}
   slot {class=OverlaySlot}
  }
  FortClickThroughButton {
   widgetClassPath=/Script/FortniteUI.FortClickThroughButton
   bIsVariable=true
   parent=ButtonContainer
   widget=FortClickThroughButton
   properties {widgetStyle {
     clickedSlateSound {resourceObject=null}
     disabled {
      bIsDynamicallyLoaded=false
      drawAs=NoDrawType
      imageType=NoImage
      mirroring=NoMirror
      resourceName=None
      resourceObject=null
      tiling=NoTile
      imageSize=<0 0>
      margin=<ltrb 0 0 0 0>
      outlineSettings {
       bUseBrushTransparency=false
       roundingType=HalfHeightRadius
       width=0
       color {colorUseRule=UseColor_Specified, specifiedColor=<rgba 0 0 0 0>}
       cornerRadii=<0 0 0 0>
      }
      tintColor {colorUseRule=UseColor_Specified, specifiedColor=<rgba 1 1 1 1>}
      uVRegion {bIsValid=false, max=<0 0>, min=<0 0>}
     }
     disabledForeground {colorUseRule=UseColor_Specified, specifiedColor=<rgba 0.5271 0.5271 0.5271 1>}
     hovered {
      bIsDynamicallyLoaded=false
      drawAs=NoDrawType
      imageType=NoImage
      mirroring=NoMirror
      resourceName=None
      resourceObject=null
      tiling=NoTile
      imageSize=<32 32>
      margin=<ltrb 0 0 0 0>
      outlineSettings {
       bUseBrushTransparency=true
       roundingType=FixedRadius
       width=1
       color {colorUseRule=UseColor_Specified, specifiedColor=<rgba 0.7243 0.7243 0.7243 1>}
       cornerRadii=<4 4 4 4>
      }
      tintColor {colorUseRule=UseColor_Specified, specifiedColor=<rgba 0.7243 0.7243 0.7243 1>}
      uVRegion {bIsValid=false, max=<0 0>, min=<0 0>}
     }
     hoveredForeground {colorUseRule=UseColor_Specified, specifiedColor=<rgba 0 1 1 1>}
     hoveredSlateSound {resourceObject=null}
     normal {
      bIsDynamicallyLoaded=false
      drawAs=NoDrawType
      imageType=NoImage
      mirroring=NoMirror
      resourceName=None
      resourceObject=null
      tiling=NoTile
      imageSize=<32 32>
      margin=<ltrb 0 0 0 0>
      outlineSettings {
       bUseBrushTransparency=true
       roundingType=FixedRadius
       width=1
       color {colorUseRule=UseColor_Specified, specifiedColor=<rgba 0.6951 0.6951 0.6951 1>}
       cornerRadii=<4 4 4 4>
      }
      tintColor {colorUseRule=UseColor_Specified, specifiedColor=<rgba 0.4955 0.4955 0.4955 1>}
      uVRegion {bIsValid=false, max=<0 0>, min=<0 0>}
     }
     normalForeground {colorUseRule=UseColor_Specified, specifiedColor=<rgba 0 1 1 1>}
     normalPadding=<ltrb 0 0 0 0>
     pressed {
      bIsDynamicallyLoaded=false
      drawAs=NoDrawType
      imageType=NoImage
      mirroring=NoMirror
      resourceName=None
      resourceObject=null
      tiling=NoTile
      imageSize=<32 32>
      margin=<ltrb 0 0 0 0>
      outlineSettings {
       bUseBrushTransparency=true
       roundingType=FixedRadius
       width=1
       color {colorUseRule=UseColor_Specified, specifiedColor=<rgba 0.7243 0.7243 0.7243 1>}
       cornerRadii=<4 4 4 4>
      }
      tintColor {colorUseRule=UseColor_Specified, specifiedColor=<rgba 0.3843 0.3843 0.3843 1>}
      uVRegion {bIsValid=false, max=<0 0>, min=<0 0>}
     }
     pressedForeground {colorUseRule=UseColor_Specified, specifiedColor=<rgba 0 1 1 1>}
     pressedPadding=<ltrb 0 0 0 0>
     pressedSlateSound {resourceObject=null}
    }}
   slot {class=OverlaySlot, properties {horizontalAlignment=HAlign_Fill, verticalAlignment=VAlign_Fill}}
  }
  EntryBorder {
   widgetClassPath=&Border
   bIsVariable=true
   parent=FortClickThroughButton
   widget=EntryBorder
   properties {
    visibility=SelfHitTestInvisible
    background {
     bIsDynamicallyLoaded=false
     drawAs=Box
     imageType=NoImage
     mirroring=NoMirror
     resourceName=None
     resourceObject=/Game/UI/Sidebar/Materials/M-Button-RoundCorner_SidebarInteractionRound
     tiling=NoTile
     imageSize=<112 112>
     margin=<ltrb 0.5 0.5 0.5 0.5>
     outlineSettings {
      bUseBrushTransparency=false
      roundingType=HalfHeightRadius
      width=0
      color {colorUseRule=UseColor_Specified, specifiedColor=<rgba 0 0 0 0>}
      cornerRadii=<0 0 0 0>
     }
     tintColor {colorUseRule=UseColor_Specified, specifiedColor=<rgba 1 1 1 1>}
     uVRegion {bIsValid=false, max=<0 0>, min=<0 0>}
    }
    brushColor=<rgba 0.004 0.063 0.5029 1>
    padding=<ltrb 50 20 50 20>
   }
   slot {class=ButtonSlot}
  }
  ContentContainer_Padding {
   widgetClassPath=&HorizontalBox
   parent=EntryBorder
   widget=ContentContainer_Padding
   properties {flowDirectionPreference=Culture}
   slot {class=BorderSlot, properties {padding=<ltrb 50 20 50 20>}}
  }
  PaddingForText2x {
   widgetClassPath=&Spacer
   parent=ContentContainer_Padding
   widget=PaddingForText2x
   properties {size=<48 1>}
   slot {class=HorizontalBoxSlot}
  }
  MinMaxWidth {
   widgetClassPath=&SizeBox
   parent=ContentContainer_Padding
   widget=MinMaxWidth
   properties {
    bOverride_MaxDesiredWidth=true
    bOverride_MinDesiredHeight=true
    bOverride_MinDesiredWidth=true
    heightOverride=72
    maxDesiredWidth=400
    minDesiredHeight=72
    minDesiredWidth=250
   }
   slot {class=HorizontalBoxSlot, properties {verticalAlignment=VAlign_Center, size {sizeRule=Fill, value=1}}}
  }
  TextColorBorder {
   widgetClassPath=&CommonBorder
   parent=MinMaxWidth
   widget=TextColorBorder
   properties {horizontalAlignment=HAlign_Center, verticalAlignment=VAlign_Center, contentColorAndOpacity=<rgba 1 1 1 0.5>, padding=<ltrb 0 0 0 0>}
   slot {class=SizeBoxSlot}
  }
  ButtonText {
   widgetClassPath=&CommonTextBlock
   bInherited=true
   bIsVariable=true
   parent=TextColorBorder
   widget=ButtonText
   properties {
    bIsScrollingEnabled=false
    clipping=ClipToBounds
    justification=Center
    mobileFontSizeMultiplier=1.32
    scrollStyle=/FortUILibrary/Styles/Typography/Foundation/TextScrollStyle-NoFade.TextScrollStyle-NoFade_C
    text=Centered
    textOverflowPolicy=Ellipsis
    textTransformPolicy=ToUpper
    font {
     bForceMonospaced=false
     bMaterialIsStencil=false
     fontMaterial=null
     fontObject=&BurbankSmall
     letterSpacing=0
     monospacedWidth=1
     size=18
     skewAmount=0
     typefaceFontName=Black
     outlineSettings {
      bApplyOutlineToDropShadows=false
      bMiteredCorners=false
      bSeparateFillAlpha=false
      outlineMaterial=null
      outlineSize=0
      outlineColor=<rgba 0 0 0 1>
     }
    }
    shadowOffset=<0 0>
   }
   slot {class=BorderSlot, properties {horizontalAlignment=HAlign_Center, verticalAlignment=VAlign_Center, padding=<ltrb 0 0 0 0>}}
  }
  PaddingForText {
   widgetClassPath=&Spacer
   parent=ContentContainer_Padding
   widget=PaddingForText
   properties {size=<16 1>}
   slot {class=HorizontalBoxSlot}
  }
  ForceSpaceContainer {
   widgetClassPath=&Overlay
   parent=ContentContainer_Padding
   widget=ForceSpaceContainer
   slot {class=HorizontalBoxSlot, properties {horizontalAlignment=HAlign_Center, verticalAlignment=VAlign_Center}}
  }
  InputActionWidget {
   widgetClassPath=&CommonActionWidget
   bInherited=true
   bIsVariable=true
   parent=ForceSpaceContainer
   widget=InputActionWidget
   properties {renderOpacity=0, visibility=SelfHitTestInvisible, inputActions [{dataTable=&InputActionDataTable, rowName=Select}]}
   slot {class=OverlaySlot}
  }
  ForceSpace_32x32 {widgetClassPath=&Spacer, parent=ForceSpaceContainer, widget=ForceSpace_32x32, properties {size=<32 32>}, slot {class=OverlaySlot}}
 ]
}
